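# Skyfire HUD widget that shows a tier rating as a single star icon, built on SkyfireWidgetBase.
# Construct caches the star image's dynamic material instance.
# SetVisualsToTierLevel writes the given fill, outline and glow colors to that material through named vector parameters.
# Two scale animations, a bouncy pop for filled and a horizontal squeeze for empty, are defined but not played in this graph.
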WidgetBlueprint /Skyfire/UI/Skyfire_HUDElement_TierRating
animations [
 EmphasizeUpdateFilled {
  binding_count=1
  length_seconds=2
  object=&self:EmphasizeUpdateFilled
  sequence {bindings [Icon_TierIndicator_Star {
     id=#1
     possessed_class=Image
     tracks [{
       type=MovieScene2DTransformTrack
       sections [{
         type=MovieScene2DTransformSection
         end_frame=40
         start_frame=0
         channels [
          Translation.X {type=MovieSceneScriptingFloatChannel}
          Translation.Y {type=MovieSceneScriptingFloatChannel}
          Angle {type=MovieSceneScriptingFloatChannel}
          Scale.X {type=MovieSceneScriptingFloatChannel, keys [[0,1],[2,0.1],[7,1.5],[10,1],[11,0.9],[14,1],[40,1]]}
          Scale.Y {type=MovieSceneScriptingFloatChannel, keys [[0,1],[2,0.85],[7,1.5],[10,1],[11,0.9],[14,1],[40,1]]}
          Shear.X {type=MovieSceneScriptingFloatChannel}
          Shear.Y {type=MovieSceneScriptingFloatChannel}
         ]
        }]
      }]
    }]}
 }
 EmphasizeUpdateEmpty {
  binding_count=1
  length_seconds=2.05
  object=&self:EmphasizeUpdateEmpty
  sequence {bindings [Icon_TierIndicator_Star {
     id=#2
     possessed_class=Image
     tracks [{
       type=MovieScene2DTransformTrack
       sections [{
         type=MovieScene2DTransformSection
         end_frame=41
         start_frame=0
         channels [
          Translation.X {type=MovieSceneScriptingFloatChannel}
          Translation.Y {type=MovieSceneScriptingFloatChannel}
          Angle {type=MovieSceneScriptingFloatChannel}
          Scale.X {type=MovieSceneScriptingFloatChannel, keys [[0,1],[9,1],[12,0.5],[20,0.1],[40,1]]}
          Scale.Y {type=MovieSceneScriptingFloatChannel, const [1,2]}
          Shear.X {type=MovieSceneScriptingFloatChannel}
          Shear.Y {type=MovieSceneScriptingFloatChannel}
         ]
        }]
      }]
    }]}
 }
]
graphs [EventGraph {nodes [
   n1 = SetVisualsToTierLevel!() -> (
    OutputDelegate
    then
    FillColor:Linear Color Structure=(R=0,G=0,B=0,A=1)
    OutlineColor:Linear Color Structure
    GlowColor:Linear Color Structure
   ) {is=title K2Node_CustomEvent}
   n2 = Get Ref_TierIndicatorDynamicMaterial() -> (Ref_TierIndicatorDynamicMaterial:Material Instance Dynamic Object Reference)
   > SetVectorParameterValue(self=@n2, ParameterName=Get MaterialParameterName_Fill() -> (MaterialParameterName_Fill:Name=None), Value=@n1.FillColor)
   > SetVectorParameterValue(self=@n2, ParameterName=Get MaterialParameterName_Outline() -> (MaterialParameterName_Outline:Name=None), Value=@n1.OutlineColor)
   > SetVectorParameterValue(self=@n2, ParameterName=Get MaterialParameterName_Glow() -> (MaterialParameterName_Glow:Name=None), Value=@n1.GlowColor)
   Event Construct()
   > n3 = GetDynamicMaterial(self=Get Icon_TierIndicator_Star() -> (Icon_TierIndicator_Star:Image Object Reference))
   > Set Ref_TierIndicatorDynamicMaterial(Ref_TierIndicatorDynamicMaterial:Material Instance Dynamic Object Reference=@n3) -> (Output_Get:Material Instance Dynamic Object Reference)
  ]}]
variables [
 MaterialParameterName_Fill {type=string, default=ColorFill}
 MaterialParameterName_Glow {type=string, default=ColorGlow}
 MaterialParameterName_Outline {type=string, default=ColorOutline}
 Ref_TierIndicatorDynamicMaterial {type=MaterialInstanceDynamic}
]
widgets {
 info {inheritedWidgetCount=1, parentClass=/Script/SkyfireUI.SkyfireWidgetBase, rootWidgetClass=&Overlay, widgetCount=3}
 widgets [
  Overlay_15 {widgetClassPath=&Overlay, widget=Overlay_15}
  Icon_TierIndicator_Star {
   widgetClassPath=&Image
   bIsVariable=true
   parent=Overlay_15
   widget=Icon_TierIndicator_Star
   properties {brush {
     bIsDynamicallyLoaded=false
     drawAs=Image
     imageType=NoImage
     mirroring=NoMirror
     resourceName=None
     resourceObject=/Skyfire/UI/Art/MI_UI_Skyfire_StarTierIndicator_Inst
     tiling=NoTile
     imageSize=<52 52>
     margin=<ltrb 0 0 0 0>
     outlineSettings {
      bUseBrushTransparency=false
      roundingType=HalfHeightRadius
      width=0
      color {colorUseRule=UseColor_Specified, specifiedColor=<rgba 0 0 0 0>}
      cornerRadii=<0 0 0 0>
     }
     tintColor {colorUseRule=UseColor_Specified, specifiedColor=<rgba 1 1 1 1>}
     uVRegion {bIsValid=false, max=<0 0>, min=<0 0>}
    }}
   slot {class=OverlaySlot}
  }
  TutorialHighlight {widgetClassPath=&TutorialHighlight}
 ]
}
