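# Skyfire HUD widget that shows the current Vault Tier as a row of five stars and announces tier changes.
# OnTierLevelUpdated plays a "+star" pop animation, then maps the rarity to a star count through a lookup map.
# Each star is recolored with the rarity color, filled up to that count, and plays a filled or empty emphasis animation.
# Rarities missing from the map hide the stars and show the tinted rarity name as fallback text.
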WidgetBlueprint /Skyfire/UI/Skyfire_HUD_TierMessaging
animations [StarAddedMessaging {
  binding_count=3
  length_seconds=3.5
  object=&self:StarAddedMessaging
  sequence {bindings [
    HorizontalBox_NewStarMessage {
     id=#1
     possessed_class=HorizontalBox
     tracks [{
       type=MovieSceneByteTrack
       display_name=Visibility
       sections [{type=MovieSceneByteSection, end_frame=69, start_frame=0, channels [None {type=MovieSceneScriptingByteChannel, keys [[0,4],[69,2]]}]}]
      }]
    }
    Skyfire_HUDElement_TierStar {
     id=#2
     possessed_class=Skyfire_HUDElement_TierRating_C
     tracks [{
       type=MovieScene2DTransformTrack
       sections [{
         type=MovieScene2DTransformSection
         end_frame=70
         start_frame=0
         channels [
          Translation.X {type=MovieSceneScriptingFloatChannel}
          Translation.Y {type=MovieSceneScriptingFloatChannel}
          Angle {type=MovieSceneScriptingFloatChannel}
          Scale.X {type=MovieSceneScriptingFloatChannel, keys [[0,0],[3,0],[9,1.5],[13,1],[60,1],[69,0]]}
          Scale.Y {type=MovieSceneScriptingFloatChannel, keys [[0,0],[3,1],[9,1.5],[13,1],[60,1],[69,0.5]]}
          Shear.X {type=MovieSceneScriptingFloatChannel}
          Shear.Y {type=MovieSceneScriptingFloatChannel}
         ]
        }]
      }]
    }
    RichTextBlock_PlusSign {
     id=#3
     possessed_class=RichTextBlock
     tracks [{
       type=MovieSceneFloatTrack
       display_name=Render Opacity
       sections [{
         type=MovieSceneFloatSection
         end_frame=69
         start_frame=0
         channels [None {type=MovieSceneScriptingFloatChannel, keys [[0,0],[5,1],[60,1],[69,0]]}]
        }]
      }]
    }
   ]}
 }]
graphs [EventGraph {nodes [
   n1 = OnTierLevelUpdated!() -> (OutputDelegate, then, UpdatedRarityTier:EFortRarity Enum=Common) {is=title K2Node_CustomEvent}
   > PlayAnimation(InAnimation=Get StarAddedMessaging() -> (StarAddedMessaging:Widget Animation Object Reference))
   > n2 = GetRarityData!(execute, Rarity:EFortRarity Enum=@n1.UpdatedRarityTier) -> (then, ReturnValue:Fort Rarity Item Data Structure)
   n3 = Find(
    TargetMap:Map of EFortRarity Enums to Integers=Get RarityEnumToNumOfStarsMap() -> (RarityEnumToNumOfStarsMap:Map of EFortRarity Enums to Integers=Common)
    Key:"EFortRarity Enum (by ref)"=@n1.UpdatedRarityTier
   ) -> (Value:int)
   > n4 = Branch(Condition=@n3.ReturnValue)
   > n5 = For Each Loop(Array:Array of Skyfire HUDElement Tier Rating Object References=Get TierIndicatorStarWidgets() -> (TierIndicatorStarWidgets:Array of Skyfire HUDElement Tier Rating Object References)) -> (Array Element:Skyfire HUDElement Tier Rating Object Reference)
   n6 = Break Fort Rarity Item Data!(FortRarityItemData:"Fort Rarity Item Data Structure (by ref)"=@n2) -> (
    Name:Text
    Color1:Linear Color Structure
    Color2:Linear Color Structure
    Color3:Linear Color Structure
    Color4:Linear Color Structure
    Color5:Linear Color Structure
    Radius:float=~
    Falloff:float=~
    Brightness:float=~
    Roughness:float=~
    Glow:float=~
   ) {is=title K2Node_BreakStruct}
   n7 = "integer < integer"!(A:int=@n5.Array Index, B:int=@n3) -> (ReturnValue:bool) {is=title K2Node_CallFunction}
   > SetVisualsToTierLevel!(
    execute
    self:Skyfire HUDElement Tier Rating Object Reference=@n5
    FillColor:Linear Color Structure=SelectColor(A=@n6.Color1, B=Get Color_TierUnobtainedFill() -> (Color_TierUnobtainedFill:Linear Color Structure), bPickA=@n7)
    OutlineColor:Linear Color Structure=@n6.Color1
    GlowColor:Linear Color Structure=@n6.Color1
   ) -> (then)
   > PlayAnimation(
    self=@n5
    InAnimation=Select(
     Option 0:Widget Animation Object Reference=Get EmphasizeUpdateEmpty!(self:Skyfire HUDElement Tier Rating Object Reference=@n5) -> (EmphasizeUpdateEmpty:Widget Animation Object Reference)
     Option 1:Widget Animation Object Reference=Get EmphasizeUpdateFilled!(self:Skyfire HUDElement Tier Rating Object Reference=@n5) -> (EmphasizeUpdateFilled:Widget Animation Object Reference)
     Index=@n7
    ) -> (ReturnValue:Widget Animation Object Reference)
   )
   SetVisibility(
    execute=@n4.else
    self=Get HorizontalBox_StarIndicators() -> (HorizontalBox_StarIndicators:Horizontal Box Object Reference)
    InVisibility=Hidden
   )
   n8 = Get CommonText_FallbackDisplay() -> (CommonText_FallbackDisplay:Common Text Object Reference)
   > SetVisibility(self=@n8, InVisibility=SelfHitTestInvisible)
   > SetColorAndOpacity(self=@n8, InColorAndOpacity=Make Slate Color(SpecifiedColor=@n6.Color1))
   > "SetText (Text)"(self=@n8, InText=@n6)
   Event Construct()
   > Set TierIndicatorStarWidgets(TierIndicatorStarWidgets:Array of Skyfire HUDElement Tier Rating Object References=Make Array!(
     "[0]":Skyfire HUDElement Tier Rating Object Reference=Get Skyfire_HUDElement_TierStar1() -> (Skyfire_HUDElement_TierStar1:Skyfire HUDElement Tier Rating Object Reference)
     "[1]":Skyfire HUDElement Tier Rating Object Reference=Get Skyfire_HUDElement_TierStar2() -> (Skyfire_HUDElement_TierStar2:Skyfire HUDElement Tier Rating Object Reference)
     "[2]":Skyfire HUDElement Tier Rating Object Reference=Get Skyfire_HUDElement_TierStar3() -> (Skyfire_HUDElement_TierStar3:Skyfire HUDElement Tier Rating Object Reference)
     "[3]":Skyfire HUDElement Tier Rating Object Reference=Get Skyfire_HUDElement_TierStar4() -> (Skyfire_HUDElement_TierStar4:Skyfire HUDElement Tier Rating Object Reference)
     "[4]":Skyfire HUDElement Tier Rating Object Reference=Get Skyfire_HUDElement_TierStar5() -> (Skyfire_HUDElement_TierStar5:Skyfire HUDElement Tier Rating Object Reference)
    ) -> (Array:Array of Skyfire HUDElement Tier Rating Object References)) -> (Output_Get:Array of Skyfire HUDElement Tier Rating Object References)
   SelectColor(A=@n6.Color1, B=Get Color_TierReachedFill() -> (Color_TierReachedFill:Linear Color Structure), bPickA=@n7)
  ]}]
variables [
 Color_TierReachedFill {type=LinearColor, default=<rgba 1 1 1 1>}
 Color_TierUnobtainedFill {type=LinearColor, default=<rgba 0 0 0 0>}
 RarityEnumToNumOfStarsMap {default {common=1, epic=4, legendary=5, rare=3, uncommon=2}}
 TierIndicatorStarWidgets {type="array<Skyfire_HUDElement_TierRating_C>", default []}
]
widgets {
 info {inheritedWidgetCount=1, parentClass=/Script/SkyfireUI.SkyfireWidgetBase, rootWidgetClass=&Overlay, widgetCount=15}
 widgets [
  Overlay_TierMessaging {widgetClassPath=&Overlay, widget=Overlay_TierMessaging}
  VerticalBox_31 {widgetClassPath=&VerticalBox, parent=Overlay_TierMessaging, widget=VerticalBox_31, slot {class=OverlaySlot}}
  HorizontalBox_NewStarMessage {
   widgetClassPath=&HorizontalBox
   parent=VerticalBox_31
   widget=HorizontalBox_NewStarMessage
   properties {visibility=Hidden, renderTransform {angle=0, scale=<1 1>, shear=<-8 0>, translation=<0 0>}}
   slot {class=VerticalBoxSlot, properties {horizontalAlignment=HAlign_Center, verticalAlignment=VAlign_Top}}
  }
  RichTextBlock_PlusSign {
   widgetClassPath=/Script/UMG.RichTextBlock
   bIsVariable=true
   parent=HorizontalBox_NewStarMessage
   widget=RichTextBlock_PlusSign
   properties {text="<HUDTierLabel>+</>", textStyleSet=&/SkyfireTextStyles}
   slot {class=HorizontalBoxSlot}
  }
  Skyfire_HUDElement_TierStar {
   widgetClassPath=&/Skyfire_HUDElement_TierRating
   bIsVariable=true
   parent=HorizontalBox_NewStarMessage
   widget=Skyfire_HUDElement_TierStar
   slot {class=HorizontalBoxSlot, properties {horizontalAlignment=HAlign_Center, verticalAlignment=VAlign_Center}}
  }
  Text_TierLabel {
   widgetClassPath=&RichTextBlock
   bIsVariable=true
   parent=VerticalBox_31
   widget=Text_TierLabel
   properties {
    text="<HUDTierLabel>VAULT TIER</>"
    textStyleSet=&SkyfireTextStyles
    renderTransform {angle=0, scale=<1 1>, shear=<-8 0>, translation=<0 0>}
   }
   slot {class=VerticalBoxSlot, properties {horizontalAlignment=HAlign_Center, verticalAlignment=VAlign_Bottom}}
  }
  Overlay_93 {widgetClassPath=&Overlay, parent=VerticalBox_31, widget=Overlay_93, slot {class=VerticalBoxSlot}}
  HorizontalBox_StarIndicators {
   widgetClassPath=&HorizontalBox
   bIsVariable=true
   parent=Overlay_93
   widget=HorizontalBox_StarIndicators
   slot {class=OverlaySlot, properties {horizontalAlignment=HAlign_Center}}
  }
  Skyfire_HUDElement_TierStar1 {
   widgetClassPath=&Skyfire_HUDElement_TierRating
   bIsVariable=true
   parent=HorizontalBox_StarIndicators
   widget=Skyfire_HUDElement_TierStar1
   slot {class=HorizontalBoxSlot, properties {horizontalAlignment=HAlign_Center, verticalAlignment=VAlign_Top}}
  }
  Skyfire_HUDElement_TierStar2 {
   widgetClassPath=&Skyfire_HUDElement_TierRating
   bIsVariable=true
   parent=HorizontalBox_StarIndicators
   widget=Skyfire_HUDElement_TierStar2
   slot {class=HorizontalBoxSlot, properties {horizontalAlignment=HAlign_Center, verticalAlignment=VAlign_Top}}
  }
  Skyfire_HUDElement_TierStar3 {
   widgetClassPath=&Skyfire_HUDElement_TierRating
   bIsVariable=true
   parent=HorizontalBox_StarIndicators
   widget=Skyfire_HUDElement_TierStar3
   slot {class=HorizontalBoxSlot, properties {horizontalAlignment=HAlign_Center, verticalAlignment=VAlign_Top}}
  }
  Skyfire_HUDElement_TierStar4 {
   widgetClassPath=&Skyfire_HUDElement_TierRating
   bIsVariable=true
   parent=HorizontalBox_StarIndicators
   widget=Skyfire_HUDElement_TierStar4
   slot {class=HorizontalBoxSlot, properties {horizontalAlignment=HAlign_Center, verticalAlignment=VAlign_Top}}
  }
  Skyfire_HUDElement_TierStar5 {
   widgetClassPath=&Skyfire_HUDElement_TierRating
   bIsVariable=true
   parent=HorizontalBox_StarIndicators
   widget=Skyfire_HUDElement_TierStar5
   slot {class=HorizontalBoxSlot, properties {horizontalAlignment=HAlign_Center, verticalAlignment=VAlign_Top}}
  }
  CommonText_FallbackDisplay {
   widgetClassPath=&CommonTextBlock
   bIsVariable=true
   parent=Overlay_93
   widget=CommonText_FallbackDisplay
   properties {
    text=RATING DISPLAY
    textTransformPolicy=ToUpper
    visibility=Collapsed
    font {
     bForceMonospaced=false
     bMaterialIsStencil=false
     fontMaterial=null
     fontObject=&Burbank
     letterSpacing=0
     monospacedWidth=1
     size=28
     skewAmount=0
     typefaceFontName=BurbankRegular-Black
     outlineSettings {
      bApplyOutlineToDropShadows=false
      bMiteredCorners=false
      bSeparateFillAlpha=false
      outlineMaterial=&M-FontShadow
      outlineSize=3
      outlineColor=<rgba 0 0 0 0.5>
     }
    }
    shadowOffset=<0 0>
   }
   slot {class=OverlaySlot}
  }
  TutorialHighlight {widgetClassPath=&TutorialHighlight}
 ]
}
