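# Skyfire mode HUD widget showing the countdown to loot-room teleport and the player's earned loot rarity tier.
# Mutator phase changes and pawn-component delegates, re-registered on pawn change, initialize both displays in core gameplay.
# A looping timer refreshes and recolors the clock, collapsing the widget at zero; time changes animate a playlist-specific "+time" message.
# Mothership state also toggles visibility of map layers and full-screen map controls found by class search.
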WidgetBlueprint /Skyfire/UI/SkyfireHUD
animations [TimeAddedMessage {
  binding_count=1
  length_seconds=2.6
  object=&self:TimeAddedMessage
  sequence {bindings [RichText_TimeAddedMessage {
     id=#1
     possessed_class=RichTextBlock
     tracks [
      {
       type=MovieSceneByteTrack
       display_name=Visibility
       sections [{type=MovieSceneByteSection, end_frame=52, start_frame=0, channels [None {type=MovieSceneScriptingByteChannel, keys [[0,4],[52,1]]}]}]
      }
      {
       type=MovieSceneFloatTrack
       display_name=Render Opacity
       sections [{
         type=MovieSceneFloatSection
         end_frame=52
         start_frame=0
         channels [None {type=MovieSceneScriptingFloatChannel, keys [[0,0],[6,1],[42,1],[52,0]]}]
        }]
      }
      {
       type=MovieScene2DTransformTrack
       sections [{
         type=MovieScene2DTransformSection
         end_frame=52
         start_frame=0
         channels [
          Translation.X {type=MovieSceneScriptingFloatChannel, const [0,3]}
          Translation.Y {type=MovieSceneScriptingFloatChannel, keys [[0,60],[6,0],[52,0]]}
          Angle {type=MovieSceneScriptingFloatChannel}
          Scale.X {type=MovieSceneScriptingFloatChannel, keys [[0,1],[44,1],[52,0.7]]}
          Scale.Y {type=MovieSceneScriptingFloatChannel, keys [[0,0.2],[6,1.2],[7,1],[44,1],[52,0.7]]}
          Shear.X {type=MovieSceneScriptingFloatChannel}
          Shear.Y {type=MovieSceneScriptingFloatChannel}
         ]
        }]
      }
     ]
    }]}
 }]
graphs [
 DeterminePlaylistType {nodes [
   DeterminePlaylistType() {is=title K2Node_FunctionEntry}
   n1 = GetGameStateAthena!() -> (ReturnValue:Fort Game State Athena Object Reference)
   > Is Valid(InputObject=@n1)
   > n2 = For Each Loop with Break!(
    Exec:Exec
    Array:Array of Gameplay Tag Structures=Get PossiblePlaylistTags() -> (PossiblePlaylistTags:Array of Gameplay Tag Structures)
    Break:Exec=@n3.then
   ) -> (LoopBody:Exec, Array Element:Gameplay Tag Structure, Array Index:int, Completed:Exec)
   > Branch(Condition=Compare Tag Container to Other Tags!(
     TagCase_0:"Gameplay Tag Structure (by ref)"=@n2
     Gameplay Tag Container:"Gameplay Tag Container Structure (by ref)"=GetAthenaPlaylistContextTags!(self:Fort Game State Athena Object Reference=@n1) -> (ReturnValue:Gameplay Tag Container Structure)
    ) -> (Case_0 True:bool=~) {is=title GameplayTagsK2Node_MultiCompareGameplayTagContainerSingleTags})
   > n3 = Set PlaylistType(PlaylistType:Gameplay Tag Structure=@n2) -> (Output_Get:Gameplay Tag Structure)
  ]}
 EventGraph {nodes [
   UpdateTeleportToLootRoomTimer() {is=title K2Node_CustomEvent}
   > n1 = Get!(execute:Exec) -> (then:Exec, else:Exec, CachedSkyfirePlayerPawnComponent:Skyfire Player Pawn Component Object Reference)
   n2 = Get CommonText_Clock_TimeRemaining() -> (CommonText_Clock_TimeRemaining:Common Numeric Text Block Object Reference)
   n3 = GetTimeUntilTeleportToLootRoom!(self:Skyfire Player Pawn Component Object Reference=@n1) -> (ReturnValue:float)
   > SetCurrentValue!(execute, self:Common Numeric Text Block Object Reference=@n2, NewValue:float=@n3) -> (then)
   n4 = Get TimeUntilVaultWarningColorMaxThreshold() -> (TimeUntilVaultWarningColorMaxThreshold:int=0)
   n5 = Ceil(A=@n3) {is=title K2Node_CallFunction}
   > SetColorAndOpacity(
    self=@n2
    InColorAndOpacity=Select(
     Option 0:Slate Color Structure=Get ColorVaultTimerWarningThreshold() -> (ColorVaultTimerWarningThreshold:Slate Color Structure)
     Option 1:Slate Color Structure=Get ColorVaultTimerDefault() -> (ColorVaultTimerDefault:Slate Color Structure)
     Index="integer > integer"(A=@n5, B=@n4) {is=title K2Node_CallFunction}
    ) -> (ReturnValue:Slate Color Structure)
   )
   > Branch(Condition="integer <= integer"!(A:int=@n5, B:int=0) -> (ReturnValue:bool) {is=title K2Node_CallFunction})
   > Clear and Invalidate Timer by Handle(Handle=Get TimerUntilSquadIsTeleportedHandle() -> (TimerUntilSquadIsTeleportedHandle:Timer Handle Structure))
   > Unbind Event from On Earned Loot Rarity Changed(
    self:Skyfire Player Pawn Component Object Reference=Get CachedSkyfirePlayerPawnComponent() -> (CachedSkyfirePlayerPawnComponent:Skyfire Player Pawn Component Object Reference)
    Delegate=Create Event()
   ) {is=title K2Node_RemoveDelegate}
   > SetVisibility(InVisibility=Collapsed)
   n6 = SkyfirePhaseChanged!() -> (OutputDelegate, then, NewPhase:ESkyfirePhase Enum=None) {is=title K2Node_CustomEvent}
   > n7 = Switch on ESkyfirePhase!(execute, Selection:ESkyfirePhase Enum=@n6.NewPhase) -> (None:Exec, Prison:Exec, CoreGameplay:Exec, AllPlayersExited:Exec) {is=title K2Node_SwitchEnum}
   Set IsInCoreGameplayPhase(execute=@n7.CoreGameplay, IsInCoreGameplayPhase:bool=true) -> (Output_Get:bool)
   > MaybeInitTeleportToLootRoomTimer()
   > MaybeInitEarnedLootRarity()
   > n8 = Get!(execute) -> (then, else, CachedSkyfireMutator:Fort Athena Mutator Skyfire Object Reference)
   > Unbind Event from On Skyfire Phase Changed(self:Fort Athena Mutator Skyfire Object Reference=@n8, Delegate=Create Event()) {is=title K2Node_RemoveDelegate}
   n9 = OnPawnComponentAttachedEvent!() -> (OutputDelegate, then, OwnerPawn:Fort Pawn Object Reference, AttachedComponent:Fort Pawn Component Object Reference) {is=title K2Node_CustomEvent}
   n10 = GetOwningPlayerPawn!(self:User Widget Object Reference) -> (ReturnValue:Pawn Object Reference)
   Is Valid(exec=[@n9.then, @n14.then, @n15.then], InputObject=@n10)
   > n11 = Cast To FortPlayerPawn(Object=@n10)
   n12 = GetComponentByClass!(self:Actor Object Reference=@n11, ComponentClass:Actor Component Class Reference=/Script/SkyfireRuntime.SkyfirePlayerPawnComponent) -> (ReturnValue:Skyfire Player Pawn Component Object Reference)
   > n13 = Is Valid(InputObject=@n12)
   Unbind Event from On Pawn Component Attached Event(execute=@n13.Is Not Valid, self:Fort Pawn Object Reference=@n11, Delegate=@n9) {is=title K2Node_RemoveDelegate}
   > Assign On Pawn Component Attached Event!(execute:Exec, self:Fort Pawn Object Reference=@n11, Delegate:"Delegate (by ref)"=@n9) -> (then:Exec) {is=title K2Node_AssignDelegate}
   UnbindPlayerPawn(execute=@n13.Is Valid)
   > Set Cached Skyfire Player Pawn Component!(execute, self, InSkyfirePlayerPawnComponent:Skyfire Player Pawn Component Object Reference=@n12) -> (then) {is=title K2Node_CallFunction}
   n14 = RegisterPawnComponent() {is=title K2Node_CustomEvent}
   n15 = OnPawnChanged!() -> (OutputDelegate, then, NewPawn:Fort Pawn Object Reference) {is=title K2Node_CustomEvent}
   MaybeInitTeleportToLootRoomTimer() {is=title K2Node_CustomEvent}
   > Branch(Condition=Get IsInCoreGameplayPhase() -> (IsInCoreGameplayPhase:bool))
   > n16 = Get!(execute) -> (then, else, CachedSkyfirePlayerPawnComponent:Skyfire Player Pawn Component Object Reference)
   > Do Once()
   > Bind Event to On Teleport to Loot Room Time Changed(self:Skyfire Player Pawn Component Object Reference=@n16, Delegate=Create Event()) {is=title K2Node_AddDelegate}
   > Branch(Condition="float >= float"(A=GetTeleportToLootRoomTime!(self:Skyfire Player Pawn Component Object Reference=@n16) -> (ReturnValue:float), B=0) {is=title K2Node_CallFunction})
   > StartUpdateTeleportToLootRoomTimer()
   n17 = OnEarnedLootRarityChanged!() -> (OutputDelegate, then, NewRarity:EFortRarity Enum=Common) {is=title K2Node_CustomEvent}
   > OnTierLevelUpdated!(
    execute
    self:Skyfire HUD Tier Messaging Object Reference=Get Skyfire_HUD_TierMessaging() -> (Skyfire_HUD_TierMessaging:Skyfire HUD Tier Messaging Object Reference)
    UpdatedRarityTier:EFortRarity Enum=@n17.NewRarity
   ) -> (then)
   StartUpdateTeleportToLootRoomTimer() {is=title K2Node_CustomEvent}
   > n18 = Set Timer by Event(Delegate=Create Event(), Time=Get TimerUpdateRate() -> (TimerUpdateRate:double), bLooping=true)
   > Set TimerUntilSquadIsTeleportedHandle(TimerUntilSquadIsTeleportedHandle:Timer Handle Structure=@n18) -> (Output_Get:Timer Handle Structure)
   > Set TimerStarted(TimerStarted:bool=true) -> (Output_Get:bool)
   > n19 = Get()
   > n20 = SetCurrentValue(self=Get CommonText_Clock_TimeRemaining(), NewValue=GetTimeUntilTeleportToLootRoom(self=@n19))
   SetVisibility(execute=[@n20.then, @n19.else], InVisibility=HitTestInvisible)
   ShowTeleportToLootRoomTimerUpdatedMessaging() {is=title K2Node_CustomEvent}
   > PlayAnimation(InAnimation=Get TimeAddedMessage() -> (TimeAddedMessage:Widget Animation Object Reference))
   OnTeleportToLootRoomTimeChanged!() -> (OutputDelegate, then, NewTime:float=~) {is=title K2Node_CustomEvent}
   > n21 = Branch(Condition=Get TimerStarted() -> (TimerStarted:bool))
   > ShowTeleportToLootRoomTimerUpdatedMessaging() {is=title K2Node_CallFunction}
   StartUpdateTeleportToLootRoomTimer(execute=@n21.else) {is=title K2Node_CallFunction}
   n22 = EventOnInMothershipChanged!() -> (OutputDelegate, then, bInMothership:bool=~) {is=title K2Node_CustomEvent}
   > Update Visibilities Based On In Mothership!(execute, self, In Mothership:bool=@n22.bInMothership) -> (then)
   MaybeInitEarnedLootRarity() {is=title K2Node_CustomEvent}
   > Branch(Condition=Get IsInCoreGameplayPhase())
   > n23 = Get()
   > Do Once()
   > Bind Event to On Earned Loot Rarity Changed(self:Skyfire Player Pawn Component Object Reference=@n23, Delegate=Create Event()) {is=title K2Node_AddDelegate}
   n24 = GetEarnedLootRarity!(self:Skyfire Player Pawn Component Object Reference=@n23) -> (ReturnValue:EFortRarity Enum=Common)
   > Branch(Condition="Not Equal (Enum)"!(A:EFortRarity Enum=@n24, B:EFortRarity Enum=Common) -> (ReturnValue:bool))
   > OnEarnedLootRarityChanged!(execute, self, NewRarity:EFortRarity Enum=@n24) -> (then) {is=title K2Node_CallFunction}
   Event Construct()
   > SetVisibility(InVisibility=Collapsed)
   Event Destruct()
   > Parent: Destruct()
   > n25 = Sequence!(execute) -> (then_0, then_1, then_2:Exec)
   > n26 = Get()
   > Unbind Event from On Earned Loot Rarity Changed(self=@n26, Delegate=Create Event())
   > Unbind Event from On Teleport to Loot Room Time Changed(self:Skyfire Player Pawn Component Object Reference=@n26, Delegate=Create Event()) {is=title K2Node_RemoveDelegate}
   > Unbind Event from On in Mothership Changed(self:Skyfire Player Pawn Component Object Reference=@n26, Delegate=Create Event()) {is=title K2Node_RemoveDelegate}
   UnbindPlayerPawn(execute=@n25.then_1)
   > UnbindPlayerController()
   Clear and Invalidate Timer by Handle(execute=@n25.then_2, Handle=Get TimerUntilSquadIsTeleportedHandle())
   Event MutatorReady() {is=title K2Node_Event}
   > n27 = Get!(execute) -> (then, else, CachedSkyfireMutator:Fort Athena Mutator Skyfire Object Reference)
   > Bind Event to On Skyfire Phase Changed(self:Fort Athena Mutator Skyfire Object Reference=@n27, Delegate=Create Event()) {is=title K2Node_AddDelegate}
   > OnTierLevelUpdated(self=Get Skyfire_HUD_TierMessaging(), UpdatedRarityTier=Common)
   > DeterminePlaylistType() {is=title K2Node_CallFunction}
   n28 = Find(
    TargetMap:"Map of Gameplay Tag Structures to Float (double-precision)s"=Get PlaylistTypeToTimerValueMap() -> (PlaylistTypeToTimerValueMap:"Map of Gameplay Tag Structures to Float (double-precision)s")
    Key:"Gameplay Tag Structure (by ref)"=Get PlaylistType() -> (PlaylistType:Gameplay Tag Structure)
   ) -> (Value:double)
   n29 = GetValueAtLevel!(Input:"Scalable Float Structure (by ref)"=Get TimeOrbAddedTime() -> (TimeOrbAddedTime:Scalable Float Structure), Level:float=@n28) -> (ReturnValue:float)
   > SetText(
    self=Get RichText_TimeAddedMessage() -> (RichText_TimeAddedMessage:Rich Text Block Object Reference)
    InText=Select(
     Option 0:"Text (by ref)"=Format Text(Format="LOCTEXT(\"<HUDTimerIncreaseMessaging>+Time</>\")")
     Option 1:"Text (by ref)"=Format Text!(Format="LOCTEXT(\"<HUDTimerIncreaseMessaging>+{TimeAdded}</>\")", TimeAdded:float=@n29) -> (Result)
     Index=@n28.ReturnValue
    ) -> (ReturnValue:"Text (by ref)")
   )
   > RegisterPawnComponent() {is=title K2Node_CallFunction}
   > n30 = Cast To FortPlayerController(Object=GetOwningPlayer())
   n31 = Create Event()
   > Unbind Event from On Fort Pawn Changed Event(self:Fort Player Controller Object Reference=@n30, Delegate=@n31) {is=title K2Node_RemoveDelegate}
   > Bind Event to On Fort Pawn Changed Event(self:Fort Player Controller Object Reference=@n30, Delegate=@n31) {is=title K2Node_AddDelegate}
  ]}
 GetOwningPlayerSquadID {nodes [
   GetOwningPlayerSquadID() {is=title K2Node_FunctionEntry}
   n1 = GetOwningPlayer()
   > Is Valid(InputObject=@n1)
   > n2 = Cast To FortPlayerControllerAthena(Object=GetFortPlayerControllerFromActor!(Actor:Actor Object Reference=@n1) -> (ReturnValue:Fort Player Controller Object Reference))
   > n3 = Cast To FortPlayerStateAthena!(
    execute
    Object=GetRelevantPlayerState!(PlayerController:Player Controller Object Reference=@n2) -> (ReturnValue:Player State Object Reference)
   ) -> (then, CastFailed, AsFort Player State Athena:Fort Player State Athena Object Reference) {is=title K2Node_DynamicCast}
   > Return Node!(execute, SquadID:Byte=GetSquadId!(self:Fort Player State Athena Object Reference=@n3) -> (ReturnValue:Byte=0))
   Return Node(execute=[@n2.CastFailed, @n3.CastFailed], SquadID=0)
  ]}
 Set Cached Skyfire Player Pawn Component {nodes [
   n1 = Set Cached Skyfire Player Pawn Component!() -> (then, InSkyfirePlayerPawnComponent:Skyfire Player Pawn Component Object Reference) {is=title K2Node_FunctionEntry}
   > n2 = Sequence()
   n3 = Get CachedSkyfirePlayerPawnComponent()
   > Is Valid(InputObject=@n3)
   n4 = Create Event()
   > Unbind Event from On in Mothership Changed(self=@n3, Delegate=@n4)
   n5 = Set CachedSkyfirePlayerPawnComponent(execute=@n2.then_1, CachedSkyfirePlayerPawnComponent:Skyfire Player Pawn Component Object Reference=@n1) -> (Output_Get:Skyfire Player Pawn Component Object Reference)
   > Is Valid(InputObject=@n5)
   > Bind Event to On in Mothership Changed(self:Skyfire Player Pawn Component Object Reference=@n5, Delegate=@n4) {is=title K2Node_AddDelegate}
   > EventOnInMothershipChanged!(execute, self, bInMothership:bool=IsInMothership!(self:Skyfire Player Pawn Component Object Reference=@n5) -> (ReturnValue:bool)) -> (then) {is=title K2Node_CallFunction}
   MaybeInitTeleportToLootRoomTimer(execute=@n2.then_2) {is=title K2Node_CallFunction}
   > MaybeInitEarnedLootRarity() {is=title K2Node_CallFunction}
  ]}
 UnbindPlayerController {nodes [
   UnbindPlayerController() {is=title K2Node_FunctionEntry}
   > n1 = Cast To FortPlayerController(Object=GetOwningPlayer())
   > Unbind Event from On Fort Pawn Changed Event(self=@n1, Delegate=Create Event())
  ]}
 UnbindPlayerPawn {nodes [
   UnbindPlayerPawn() {is=title K2Node_FunctionEntry}
   n1 = GetOwningPlayerPawn()
   > Is Valid(InputObject=@n1)
   > n2 = Cast To FortPlayerPawn(Object=@n1)
   > Unbind Event from On Pawn Component Attached Event(self=@n2, Delegate=Create Event())
  ]}
 Update Visibilities Based On In Mothership {nodes [
   n1 = Update Visibilities Based On In Mothership!() -> (then, In Mothership:bool=~) {is=title K2Node_FunctionEntry}
   > n2 = Set DesiredVisibility(DesiredVisibility:ESlateVisibility Enum=Select(Option 0:ESlateVisibility Enum=Visible, Option 1:ESlateVisibility Enum=Hidden, Index=@n1) -> (ReturnValue:ESlateVisibility Enum)) -> (Output_Get:ESlateVisibility Enum=Visible)
   > n3 = Sequence!(execute) -> (then_0, then_1)
   > n4 = GetAllWidgetsOfClass!(execute, WidgetClass:User Widget Class Reference=/Script/FortniteUI.AthenaMapLayerContainer, TopLevelOnly:bool=false) -> (then, FoundWidgets:Array of Athena Map Layer Container Object References)
   > n5 = For Each Loop(Array:Array of Athena Map Layer Container Object References=@n4) -> (Array Element:Athena Map Layer Container Object Reference)
   > SetVisibility(
    self=Get AthenaMapLayer!(self:Athena Map Layer Container Object Reference=@n5) -> (AthenaMapLayer:Athena Map Layer Object Reference)
    InVisibility=@n2
   )
   n6 = GetAllWidgetsOfClass(
    execute=@n3.then_1
    WidgetClass=/FortniteCoreUI/HUD/MiniMap/BP_AthenaFullScreenMapControlsWidget.BP_AthenaFullScreenMapControlsWidget_C
    TopLevelOnly=false
   ) -> (FoundWidgets:Array of BP Athena Full Screen Map Controls Widget Object References)
   > n7 = For Each Loop(Array:Array of BP Athena Full Screen Map Controls Widget Object References=@n6) -> (Array Element:BP Athena Full Screen Map Controls Widget Object Reference)
   > SetVisibility(
    self=[
     Get HorizontalBox_Controls!(self:BP Athena Full Screen Map Controls Widget Object Reference=@n7) -> (HorizontalBox_Controls:Horizontal Box Object Reference)
     Get SizeBox_ToggleQuestIcons!(self:BP Athena Full Screen Map Controls Widget Object Reference=@n7) -> (SizeBox_ToggleQuestIcons:Size Box Object Reference)
    ]
    InVisibility=@n2
   )
  ]}
]
variables [
 CachedSkyfirePlayerPawnComponent {type=SkyfirePlayerPawnComponent, default=null}
 ColorVaultTimerDefault {type=SlateColor, default {colorUseRule=UseColor_Specified, specifiedColor=<rgba 1 1 1 1>}}
 ColorVaultTimerWarningThreshold {type=SlateColor, default {colorUseRule=UseColor_Specified, specifiedColor=<rgba 1 0 0.05 1>}}
 IsInCoreGameplayPhase {type=boolean}
 PlaylistType {type=GameplayTag, default {tagName=None}}
 PlaylistTypeToTimerValueMap {default {
   "(TagName=\"Athena.Playlist.Duo\")"=1
   "(TagName=\"Athena.Playlist.Solo\")"=0
   "(TagName=\"Athena.Playlist.Squad\")"=3
   "(TagName=\"Athena.Playlist.Trios\")"=2
  }}
 PossiblePlaylistTags {
  type="array<GameplayTag>"
  default [{tagName=Athena.Playlist.Solo}, {tagName=Athena.Playlist.Duo}, {tagName=Athena.Playlist.Trios}, {tagName=Athena.Playlist.Squad}]
 }
 TimeOrbAddedTime {
  type=ScalableFloat
  default {registryType=None, value=1, curve {curveTable=/Skyfire/DataTables/OverrideGameData_Skyfire, rowName=Skyfire.TimeOrbPickupTime}}
 }
 TimerStarted {type=boolean}
 TimerUntilSquadIsTeleportedHandle {type=TimerHandle}
 TimerUpdateRate {type=number, default=0.2}
 TimeUntilVaultWarningColorMaxThreshold {type=integer, default=10}
]
widgets {
 info {inheritedWidgetCount=1, parentClass=/Script/SkyfireUI.SkyfireWidgetBase, rootWidgetClass=&SafeZone, widgetCount=11}
 widgets [
  SafeZone_Fullscreen {widgetClassPath=&SafeZone, bIsVariable=true, widget=SafeZone_Fullscreen}
  Overlay_115 {widgetClassPath=&Overlay, parent=SafeZone_Fullscreen, widget=Overlay_115, slot {class=SafeZoneSlot}}
  HorizontalBox_1 {
   widgetClassPath=&HorizontalBox
   parent=Overlay_115
   widget=HorizontalBox_1
   slot {class=OverlaySlot, properties {verticalAlignment=VAlign_Bottom, padding=<ltrb 35 0 0 60>}}
  }
  Overlay_TimerIconAndShadow {widgetClassPath=&Overlay, parent=HorizontalBox_1, widget=Overlay_TimerIconAndShadow, slot {class=HorizontalBoxSlot}}
  Image_Icon_Timer_shadow {
   widgetClassPath=&Image
   bIsVariable=true
   parent=Overlay_TimerIconAndShadow
   widget=Image_Icon_Timer_shadow
   properties {
    brush {
     bIsDynamicallyLoaded=false
     drawAs=Image
     imageType=NoImage
     mirroring=NoMirror
     resourceName=None
     resourceObject=/Skyfire/UI/Art/T-Timer-Skyfire-128
     tiling=NoTile
     imageSize=<48 48>
     margin=<ltrb 0 0 0 0>
     outlineSettings {
      bUseBrushTransparency=false
      roundingType=HalfHeightRadius
      width=0
      color {colorUseRule=UseColor_Specified, specifiedColor=<rgba 0 0 0 0>}
      cornerRadii=<0 0 0 0>
     }
     tintColor {colorUseRule=UseColor_Specified, specifiedColor=<rgba 1 1 1 0.7>}
     uVRegion {bIsValid=false, max=<0 0>, min=<0 0>}
    }
    colorAndOpacity=<rgba 0 0 0 1>
   }
   slot {class=OverlaySlot, properties {horizontalAlignment=HAlign_Center, verticalAlignment=VAlign_Center, padding=<ltrb 2 3 0 0>}}
  }
  Image_Icon_Timer {
   widgetClassPath=&Image
   bIsVariable=true
   parent=Overlay_TimerIconAndShadow
   widget=Image_Icon_Timer
   properties {brush {
     bIsDynamicallyLoaded=false
     drawAs=Image
     imageType=NoImage
     mirroring=NoMirror
     resourceName=None
     resourceObject=&T-Timer-Skyfire-128
     tiling=NoTile
     imageSize=<48 48>
     margin=<ltrb 0 0 0 0>
     outlineSettings {
      bUseBrushTransparency=false
      roundingType=HalfHeightRadius
      width=0
      color {colorUseRule=UseColor_Specified, specifiedColor=<rgba 0 0 0 0>}
      cornerRadii=<0 0 0 0>
     }
     tintColor {colorUseRule=UseColor_Specified, specifiedColor=<rgba 1 1 1 1>}
     uVRegion {bIsValid=false, max=<0 0>, min=<0 0>}
    }}
   slot {class=OverlaySlot, properties {horizontalAlignment=HAlign_Center, verticalAlignment=VAlign_Center}}
  }
  Spacer_55 {widgetClassPath=&Spacer, parent=HorizontalBox_1, widget=Spacer_55, properties {size=<8 1>}, slot {class=HorizontalBoxSlot}}
  CommonText_Clock_TimeRemaining {
   widgetClassPath=/Script/CommonUI.CommonNumericTextBlock
   bIsVariable=true
   parent=HorizontalBox_1
   widget=CommonText_Clock_TimeRemaining
   properties {
    currentNumericValue=60
    numericType=Seconds
    text=0
    font {
     bForceMonospaced=false
     bMaterialIsStencil=false
     fontMaterial=null
     fontObject=&Burbank
     letterSpacing=0
     monospacedWidth=1
     size=56
     skewAmount=0
     typefaceFontName=BurbankRegular-Black
     outlineSettings {
      bApplyOutlineToDropShadows=false
      bMiteredCorners=false
      bSeparateFillAlpha=false
      outlineMaterial=null
      outlineSize=0
      outlineColor=<rgba 0 0 0 1>
     }
    }
    formattingSpecification {
     alwaysSign=false
     maximumFractionalDigits=0
     maximumIntegralDigits=324
     minimumFractionalDigits=0
     minimumIntegralDigits=3
     roundingMode=FromZero
     useGrouping=false
    }
    renderTransform {angle=0, scale=<1 1>, shear=<-8 0>, translation=<0 0>}
    shadowColorAndOpacity=<rgba 0 0 0 0.7>
    shadowOffset=<3 3>
   }
   slot {class=HorizontalBoxSlot, properties {horizontalAlignment=HAlign_Center, verticalAlignment=VAlign_Center}}
  }
  RichText_TimeAddedMessage {
   widgetClassPath=/Script/UMG.RichTextBlock
   bIsVariable=true
   parent=HorizontalBox_1
   widget=RichText_TimeAddedMessage
   properties {
    text="<HUDTimerIncreaseMessaging>+00</>"
    textStyleSet=&/SkyfireTextStyles
    textTransformPolicy=ToUpper
    visibility=Collapsed
    renderTransform {angle=0, scale=<1 1>, shear=<-6 0>, translation=<0 0>}
   }
   slot {class=HorizontalBoxSlot, properties {verticalAlignment=VAlign_Center, padding=<ltrb 0 0 0 14>}}
  }
  Skyfire_HUD_TierMessaging {
   widgetClassPath=&/Skyfire_HUD_TierMessaging
   bIsVariable=true
   parent=Overlay_115
   widget=Skyfire_HUD_TierMessaging
   slot {class=OverlaySlot, properties {horizontalAlignment=HAlign_Right, verticalAlignment=VAlign_Bottom, padding=<ltrb 0 0 35 75>}}
  }
  TutorialHighlight {widgetClassPath=&TutorialHighlight}
 ]
}
