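# Positive-feedback voice cue for the Skyfire backpack, using the narrator VO sound class with backpack attenuation and a VO concurrency group.
# A mixer layers a pitch-varied walkie-talkie "on" sound with a louder, briefly delayed random announcement line.
# An inner quality switch gives its first two levels eight line variants and the last two only three.
# The root quality switch connects only its first slot, leaving the others empty.
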SoundCue /Skyfire/Sound/VO/Skyfire_VO_Backpack_Feedback_Positive_Cue
root=QualityLevel_1
nodes [
 {id=Delay_0, children [QualityLevel_0], properties {delayMax=0.2, delayMin=0.2}}
 {id=Mixer_0, children [Modulator_0,Delay_0], properties {inputVolume [0.75,1.35]}}
 {id=Modulator_0, children [WavePlayer_12], properties {pitchMax=1.1, pitchMin=0.85, volumeMax=1, volumeMin=1}}
 {id=QualityLevel_0, children [Random_0,Random_0,Random_2,Random_2]}
 {id=QualityLevel_1, children [Mixer_0,null,null,null]}
 {
  id=Random_0
  children [WavePlayer_1,WavePlayer_2,WavePlayer_3,WavePlayer_4,WavePlayer_5,WavePlayer_6,WavePlayer_7,WavePlayer_8]
  properties {bShouldExcludeFromBranchCulling=true, weights [1,1,1,1,1,1,1,1]}
 }
 {id=Random_2, children [WavePlayer_9, WavePlayer_10, WavePlayer_11], properties {bShouldExcludeFromBranchCulling=true, weights [1,1,1]}}
 {id=WavePlayer_1, properties {soundWaveAssetPtr=&/Skyfire_Announcement_Feedback_Positive_01}}
 {id=WavePlayer_10, properties {soundWaveAssetPtr=&/Skyfire_Announcement_Feedback_Positive_02}}
 {id=WavePlayer_11, properties {soundWaveAssetPtr=&/Skyfire_Announcement_Feedback_Positive_03}}
 {id=WavePlayer_12, properties {soundWaveAssetPtr=&/Skyfire_Announcement_Walkie_Talkie_On_01}}
 {id=WavePlayer_2, properties {soundWaveAssetPtr=&Skyfire_Announcement_Feedback_Positive_02}}
 {id=WavePlayer_3, properties {soundWaveAssetPtr=&Skyfire_Announcement_Feedback_Positive_03}}
 {id=WavePlayer_4, properties {soundWaveAssetPtr=&/Skyfire_Announcement_Feedback_Positive_04}}
 {id=WavePlayer_5, properties {soundWaveAssetPtr=&/Skyfire_Announcement_Feedback_Positive_05}}
 {id=WavePlayer_6, properties {soundWaveAssetPtr=&/Skyfire_Announcement_Feedback_Positive_06}}
 {id=WavePlayer_7, properties {soundWaveAssetPtr=&/Skyfire_Announcement_Feedback_Positive_07}}
 {id=WavePlayer_8, properties {soundWaveAssetPtr=&/Skyfire_Announcement_Feedback_Positive_08}}
 {id=WavePlayer_9, properties {soundWaveAssetPtr=&Skyfire_Announcement_Feedback_Positive_01}}
]
properties {
 attenuationSettings=/Skyfire/Sound/Skyfire_Backpack_Att
 bBypassVolumeScaleForPriority=true
 maxDistance=1350
 priority=0
 soundClassObject=/Game/Sounds/FortSoundClassesAndModes/Fort_VO_Narrator
 virtualizationMode=Disabled
 volumeMultiplier=0.8
 concurrencySet [&/Skyfire_VO_Concurrency_Group]
}
