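# Widget for one alien glyph, apparently a character in an animated message of Skyfire's UI.
# On construct it creates a dynamic material and picks a random glyph from the Rows by Columns grid.
# PlayIntro and PlayWipeAnim set timers staggered by GlyphIndex, with shorter steps when bLongMessage is set.
# DisplayGlyph fades in while squashing vertical scale to normal, and OpacityWipeAnimation sweeps the material's Swipe parameter.
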WidgetBlueprint /Skyfire/UI/Skyfire_AlienGlyph
animations [
 OpacityWipeAnimation {
  binding_count=1
  length_seconds=0.25
  object=&self:OpacityWipeAnimation
  sequence {bindings [AlienGlyph {
     id=#1
     possessed_class=Image
     tracks [{
       type=MovieSceneWidgetMaterialTrack
       sections [{type=MovieSceneParameterSection, end_frame=5, start_frame=0, channels [Swipe {type=MovieSceneScriptingFloatChannel, keys [[0,-1],[5,1]]}]}]
      }]
    }]}
 }
 DisplayGlyph {
  binding_count=2
  length_seconds=0.25
  object=&self:DisplayGlyph
  sequence {bindings [
    AlienGlyph {
     id=#2
     possessed_class=Image
     tracks [
      {
       type=MovieScene2DTransformTrack
       sections [{
         type=MovieScene2DTransformSection
         end_frame=5
         start_frame=0
         channels [
          Translation.X {type=MovieSceneScriptingFloatChannel, const [0,1]}
          Translation.Y {type=MovieSceneScriptingFloatChannel, const [0,1]}
          Angle {type=MovieSceneScriptingFloatChannel, const [0,1]}
          Scale.X {type=MovieSceneScriptingFloatChannel, const [1,2]}
          Scale.Y {type=MovieSceneScriptingFloatChannel, keys [[0,2],[5,1]]}
          Shear.X {type=MovieSceneScriptingFloatChannel, const [0,1]}
          Shear.Y {type=MovieSceneScriptingFloatChannel, const [0,1]}
         ]
        }]
      }
      {
       type=MovieSceneWidgetMaterialTrack
       sections [{type=MovieSceneParameterSection, end_frame=5, start_frame=0, channels [Swipe {type=MovieSceneScriptingFloatChannel, const [-1,2]}]}]
      }
      {
       type=MovieSceneEventTrack
       sections [{
         type=MovieSceneEventTriggerSection
         channels [None {type=MovieSceneScriptingEventChannel, const ["<Struct 'MovieSceneEvent' (0x4693601f320) {}>", 1]}]
        }]
      }
     ]
    }
    Skyfire_AlienGlyph_C_212 {
     id=#3
     possessed_class=Skyfire_AlienGlyph_C
     tracks [{
       type=MovieSceneFloatTrack
       display_name=Render Opacity
       sections [{type=MovieSceneFloatSection, end_frame=5, start_frame=0, channels [None {type=MovieSceneScriptingFloatChannel, keys [[0,0],[5,1]]}]}]
      }]
    }
   ]}
 }
]
graphs [
 EventGraph {nodes [
   PlayIntro() {is=title K2Node_CustomEvent}
   > SetVisibility(InVisibility=HitTestInvisible)
   > SetRenderOpacity(InOpacity=0)
   n1 = Get bLongMessage() -> (bLongMessage:bool)
   > n2 = Set Timer by Event(
    Delegate=Create Event()
    Time=0.001
    bLooping=false
    InitialStartDelay=int * float!(A:int=Get GlyphIndex() -> (GlyphIndex:int=0), B:double=Select(Option 0:double=0.1, Option 1:double=0.06, Index=@n1) -> (ReturnValue:double)) -> (ReturnValue:double)
   )
   > Set RandomGlyphTimerHandle(RandomGlyphTimerHandle:Timer Handle Structure=@n2) -> (Output_Get:Timer Handle Structure)
   Event Construct()
   > InitializeGlyph()
   Event Destruct()
   > Clear and Invalidate Timer by Handle(Handle=Get RandomGlyphTimerHandle() -> (RandomGlyphTimerHandle:Timer Handle Structure))
   Event PreConstruct()
   Event Tick()
  ]}
 HideWidget {nodes [
   HideWidget() {is=title K2Node_FunctionEntry}
   > SetVisibility(InVisibility=Collapsed)
  ]}
 InitializeGlyph {nodes [
   InitializeGlyph() {is=title K2Node_FunctionEntry}
   > n1 = CreateDynamicMaterialInstance!(execute, Parent=Get GlyphMaterial() -> (GlyphMaterial:Material Interface Object Reference), OptionalName=None, CreationFlags=Transient) -> (then, ReturnValue)
   > SetBrushFromMaterial(self=Get AlienGlyph() -> (AlienGlyph:Image Object Reference), Material=@n1)
   > SetRenderOpacity(InOpacity=0)
   > RandomizeGlyph()
  ]}
 PlayShowGlyph {nodes [
   PlayShowGlyph() {is=title K2Node_FunctionEntry}
   > PlayAnimation(InAnimation=Get DisplayGlyph() -> (DisplayGlyph:Widget Animation Object Reference))
  ]}
 PlayWipeAnim {nodes [
   PlayWipeAnim() {is=title K2Node_FunctionEntry}
   n1 = Get bLongMessage()
   > Set Timer by Event(
    Delegate=Create Event()
    Time=0.0001
    bLooping=false
    InitialStartDelay=int * float(A=Get GlyphIndex() -> (GlyphIndex=0), B=Select(Option 0=0.05, Option 1=0.03, Index=@n1))
   )
  ]}
 RandomizeGlyph {nodes [
   RandomizeGlyph() {is=title K2Node_FunctionEntry}
   > n1 = GetDynamicMaterial(self=Get AlienGlyph())
   > n2 = Get Scalar Parameter Value!(execute, self:Material Instance Dynamic Object Reference=@n1, ParameterName:Name=Rows) -> (then, ReturnValue:float) {is=title K2Node_CallFunction}
   > n3 = Get Scalar Parameter Value(self=@n1, ParameterName=Columns)
   n4 = Truncate(A=float - float(A=float * float(A=@n3, B=@n2), B=1) {is=title K2Node_CallFunction})
   > SetScalarParameterValue(self=@n1, ParameterName=Glyph, Value="To Float (Integer)"(InInt=RandomIntegerInRange!(Min:int=0, Max:int=@n4) -> (ReturnValue:int=0)))
   > SetDesiredSizeOverride!(execute, self:Image Object Reference=Get AlienGlyph(), DesiredSize:Vector 2D Structure=Get GlyphSize() -> (GlyphSize:Vector 2D Structure)) -> (then)
  ]}
 StartWipeAnim {nodes [
   StartWipeAnim() {is=title K2Node_FunctionEntry}
   > PlayAnimation(InAnimation=Get OpacityWipeAnimation() -> (OpacityWipeAnimation:Widget Animation Object Reference))
  ]}
 UpdateMaterial {nodes [
   n1 = UpdateMaterial!() -> (then, NewMaterial:Material Interface Object Reference) {is=title K2Node_FunctionEntry}
   > SetBrushFromMaterial(self=Get AlienGlyph(), Material=@n1)
  ]}
]
variables [
 bLongMessage {type=boolean}
 GlyphIndex {type=integer}
 GlyphMaterial {type=MaterialInterface, default=/Skyfire/UI/Art/M_UI_Skyfire_Glyphs_Inst}
 GlyphSize {type=Vector2D, default=<64 64>}
 RandomGlyphTimerHandle {type=TimerHandle}
]
widgets {
 info {inheritedWidgetCount=1, parentClass=/Script/SkyfireUI.SkyfireWidgetBase, rootWidgetClass=&Image, widgetCount=2}
 widgets [
  AlienGlyph {
   widgetClassPath=&Image
   bIsVariable=true
   widget=AlienGlyph
   properties {brush {
     bIsDynamicallyLoaded=false
     drawAs=Image
     imageType=NoImage
     mirroring=NoMirror
     resourceName=None
     resourceObject=&M_UI_Skyfire_Glyphs_Inst
     tiling=NoTile
     imageSize=<44 44>
     margin=<ltrb 0 0 0 0>
     outlineSettings {
      bUseBrushTransparency=false
      roundingType=HalfHeightRadius
      width=0
      color {colorUseRule=UseColor_Specified, specifiedColor=<rgba 0 0 0 0>}
      cornerRadii=<0 0 0 0>
     }
     tintColor {colorUseRule=UseColor_Specified, specifiedColor=<rgba 1 1 1 1>}
     uVRegion {bIsValid=false, max=<0 0>, min=<0 0>}
    }}
  }
  TutorialHighlight {widgetClassPath=&TutorialHighlight}
 ]
}
