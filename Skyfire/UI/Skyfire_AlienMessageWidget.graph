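# Skyfire HUD message widget that presents text as a row of alien glyphs over a glowing backing, with an animated text reveal.
# UpdateMessageDisplay sets hidden rich text, then starts timers; glyph count derives from the text's desired width, reusing or creating Skyfire_AlienGlyph children.
# Intro animates glow and glyphs; the reveal animates a Material Parameter Collection "Swipe" for the font material.
# Outro fades everything, clears the message and broadcasts completion.
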WidgetBlueprint /Skyfire/UI/Skyfire_AlienMessageWidget
animations [
 IntroAnim {
  binding_count=2
  length_seconds=0.5
  object=&self:IntroAnim
  sequence {bindings [
    RichText_Message {
     id=#1
     possessed_class=RichTextBlock
     tracks [{
       type=MovieSceneFloatTrack
       display_name=Render Opacity
       sections [{type=MovieSceneFloatSection, end_frame=10, start_frame=0, channels [None {type=MovieSceneScriptingFloatChannel, const [0,2]}]}]
      }]
    }
    BackingGlow_HBox {
     id=#2
     possessed_class=HorizontalBox
     tracks [
      {
       type=MovieSceneFloatTrack
       display_name=Render Opacity
       sections [{type=MovieSceneFloatSection, end_frame=10, start_frame=0, channels [None {type=MovieSceneScriptingFloatChannel, keys [[0,0],[10.0003,1]]}]}]
      }
      {
       type=MovieScene2DTransformTrack
       sections [{
         type=MovieScene2DTransformSection
         end_frame=10
         start_frame=0
         channels [
          Angle {type=MovieSceneScriptingFloatChannel, const [0,1]}
          Scale.X {type=MovieSceneScriptingFloatChannel, keys [[0,0],[10.0003,1.5]]}
          Scale.Y {type=MovieSceneScriptingFloatChannel, const [1.2,2]}
          Shear.X {type=MovieSceneScriptingFloatChannel, const [0,1]}
          Shear.Y {type=MovieSceneScriptingFloatChannel, const [0,1]}
         ]
        }]
      }
     ]
    }
   ]}
 }
 OutroAnim {
  binding_count=2
  length_seconds=1.25
  object=&self:OutroAnim
  sequence {bindings [
    RichText_Message {
     id=#3
     possessed_class=RichTextBlock
     tracks [{
       type=MovieSceneFloatTrack
       display_name=Render Opacity
       sections [{type=MovieSceneFloatSection, end_frame=25, start_frame=0, channels [None {type=MovieSceneScriptingFloatChannel, keys [[0,1],[8,0],[25,0]]}]}]
      }]
    }
    BackingGlow_HBox {
     id=#4
     possessed_class=HorizontalBox
     tracks [{
       type=MovieSceneFloatTrack
       display_name=Render Opacity
       sections [{type=MovieSceneFloatSection, end_frame=25, start_frame=0, channels [None {type=MovieSceneScriptingFloatChannel, keys [[0,1],[8,1],[25,0]]}]}]
      }]
    }
   ]}
 }
 RevealText_anim {
  binding_count=1
  length_seconds=1.1
  object=&self:RevealText_anim
  sequence {
   bindings [RichText_Message {
     id=#5
     possessed_class=RichTextBlock
     tracks [
      {
       type=MovieSceneFloatTrack
       display_name=Render Opacity
       sections [{
         type=MovieSceneFloatSection
         end_frame=22
         start_frame=0
         channels [None {type=MovieSceneScriptingFloatChannel, keys [[0,0],[3,0],[4,1],[22,1]]}]
        }]
      }
      {
       type=MovieSceneEventTrack
       sections [{
         type=MovieSceneEventTriggerSection
         channels [None {type=MovieSceneScriptingEventChannel, const ["<Struct 'MovieSceneEvent' (0x4693601fdc0) {}>", 1]}]
        }]
      }
     ]
    }]
   tracks [{
     type=MovieSceneMaterialParameterCollectionTrack
     display_name=Skyfire_FontSwipe_Mat_Param
     sections [{type=MovieSceneParameterSection, channels [Swipe {type=MovieSceneScriptingFloatChannel, keys [[3,1],[17,-1]]}]}]
    }]
  }
 }
]
graphs [
 BuildAlienGlyphs {nodes [
   BuildAlienGlyphs() {is=title K2Node_FunctionEntry}
   > SetRenderOpacity(self=Get BackingGlow_HBox() -> (BackingGlow_HBox:Horizontal Box Object Reference), InOpacity=0)
   > n1 = CalculateGlyphCount!(execute, self) -> (then, NumGlyphs:int=0)
   > n2 = Set CurrMessageGlyphCount(CurrMessageGlyphCount:int=@n1) -> (Output_Get:int=0)
   n3 = Get AlienGlyph_HBox() -> (AlienGlyph_HBox:Horizontal Box Object Reference)
   > n4 = Branch(Condition="integer > integer"(A=@n2, B=GetChildrenCount!(self:Panel Widget Object Reference=@n3) -> (ReturnValue:int=0)) {is=title K2Node_CallFunction})
   > CreateNewGlyphs()
   > PlayIntro()
   n5 = Get AlienGlyph_HBox()
   n6 = For Loop(
    execute=@n4.else
    FirstIndex=Get CurrMessageGlyphCount() -> (CurrMessageGlyphCount:int=0)
    LastIndex=int - int(A=GetChildrenCount(self=@n5) -> (ReturnValue=0), B=1) {is=title K2Node_CallFunction}
   )
   > SetVisibility(
    self=GetChildAt!(self:Panel Widget Object Reference=Get AlienGlyph_HBox(), Index:int=@n6) -> (ReturnValue:Widget Object Reference)
    InVisibility=Collapsed
   )
   PlayIntro(execute=@n6.Completed)
  ]}
 CalculateGlyphCount {nodes [
   CalculateGlyphCount() {is=title K2Node_FunctionEntry}
   n1 = BreakVector2D(InVec=GetDesiredSize!(self:Widget Object Reference=Get RichText_Message() -> (RichText_Message:Rich Text Block Object Reference)) -> (ReturnValue:Vector 2D Structure))
   n2 = Get GlyphSize() -> (GlyphSize:double)
   > Return Node!(execute, NumGlyphs:int=Round(A=float / float(A=@n1, B=@n2) {is=title K2Node_CallFunction}) {is=title K2Node_CallFunction})
  ]}
 CleanupMessaging {nodes [
   CleanupMessaging() {is=title K2Node_FunctionEntry}
   > Clear and Invalidate Timer by Handle(Handle=Get RevealTimerHandle() -> (RevealTimerHandle:Timer Handle Structure))
   > Clear and Invalidate Timer by Handle(Handle=Get BuildGlyphTimer() -> (BuildGlyphTimer:Timer Handle Structure))
  ]}
 CreateNewGlyphs {nodes [
   CreateNewGlyphs() {is=title K2Node_FunctionEntry}
   > n1 = For Loop(
    FirstIndex=GetChildrenCount(self=Get AlienGlyph_HBox()) -> (ReturnValue=0)
    LastIndex=int - int(A=Get CurrMessageGlyphCount() -> (CurrMessageGlyphCount=0), B=1)
   )
   n2 = Get GlyphSize()
   > n3 = Create Widget!(
    execute:Exec
    Class:User Widget Class Reference=&/Skyfire_AlienGlyph.Skyfire_AlienGlyph_C
    OwningPlayer:Player Controller Object Reference
    GlyphSize:Vector 2D Structure=MakeVector2D(X=@n2, Y=@n2)
    GlyphIndex:int=@n1
    bLongMessage:bool=False
    GlyphMaterial:Material Interface Object Reference=Get AlienGlyphMaterial() -> (AlienGlyphMaterial:Material Interface Object Reference)
   ) -> (then:Exec, ReturnValue:Skyfire Alien Glyph Object Reference)
   n4 = float / float(A=Get GlyphSize(), B=-7)
   n5 = Get GlyphSpacingAdjustment() -> (GlyphSpacingAdjustment:double)
   > SetPadding!(
    execute
    self:User Widget Object Reference=@n3
    InPadding:Margin Structure=Make Margin(Left=0, Top=0, Right=float + float(A=@n4, B=@n5) {is=title K2Node_CommutativeAssociativeBinaryOperator}, Bottom=0)
   ) -> (then)
   > AddChild(self=Get AlienGlyph_HBox(), Content=@n3)
  ]}
 EventGraph {nodes [
   n1 = HandleOutroAnimFinished() {is=title K2Node_CustomEvent}
   > Hide Glyphs()
   > SetText(self=Get RichText_Message(), InText=GetEmptyText!() -> (ReturnValue:Text))
   > Call On Outro Anim Finished!(execute:Exec, self:Skyfire Alien Message Widget Object Reference) -> (then:Exec) {is=title K2Node_CallDelegate}
   n2 = HandleIntroAnimFinished() {is=title K2Node_CustomEvent}
   > n3 = Set Timer by Event(Delegate=Create Event(), Time=Get TextRevealDelay() -> (TextRevealDelay:double), bLooping=false)
   > Set RevealTimerHandle(RevealTimerHandle:Timer Handle Structure=@n3) -> (Output_Get:Timer Handle Structure)
   n4 = HandleRevealTextFinished() {is=title K2Node_CustomEvent}
   > Call On Intro Anim Finished!(execute:Exec, self:Skyfire Alien Message Widget Object Reference) -> (then:Exec) {is=title K2Node_CallDelegate}
   Event PreConstruct()
   > SetTextStyleSet!(
    execute
    self:Rich Text Block Object Reference=Get RichText_Message()
    NewTextStyleSet:Data Table Object Reference=Get TextStyleSet() -> (TextStyleSet:Data Table Object Reference)
   ) -> (then)
   > SetText(self=Get RichText_Message(), InText=Get PreviewText() -> (PreviewText:Text))
   > SetupInitialGlyphs()
   Event Construct()
   > BindToAnimationFinished(Animation=Get IntroAnim() -> (IntroAnim:Widget Animation Object Reference), Delegate=@n2)
   > BindToAnimationFinished(Animation=Get OutroAnim() -> (OutroAnim:Widget Animation Object Reference), Delegate=@n1)
   > BindToAnimationFinished(Animation=Get RevealText_anim() -> (RevealText_anim:Widget Animation Object Reference), Delegate=@n4)
   Event Tick()
  ]}
 Hide Glyphs {nodes [
   Hide Glyphs() {is=title K2Node_FunctionEntry}
   n1 = Get AlienGlyph_HBox()
   > n2 = For Loop(LastIndex=int - int(A=GetChildrenCount(self=@n1) -> (ReturnValue=0), B=1))
   > SetVisibility(self=GetChildAt(self=Get AlienGlyph_HBox(), Index=@n2), InVisibility=Collapsed)
  ]}
 OnIntroAnimFinished {nodes [OnIntroAnimFinished() {is=title K2Node_FunctionEntry}]}
 OnOutroAnimFinished {nodes [OnOutroAnimFinished() {is=title K2Node_FunctionEntry}]}
 PlayIntro {nodes [
   PlayIntro() {is=title K2Node_FunctionEntry}
   > SetVisibility(self=Get BackingGlow_HBox(), InVisibility=HitTestInvisible)
   > PlayAnimation(InAnimation=Get IntroAnim())
   > Branch(Condition="integer > integer"(A=Get CurrMessageGlyphCount() -> (CurrMessageGlyphCount=0), B=0))
   > SetVisibility(self=Get AlienGlyph_HBox(), InVisibility=HitTestInvisible)
   > n1 = For Loop(LastIndex=int - int(A=Get CurrMessageGlyphCount() -> (CurrMessageGlyphCount=0), B=1))
   > n2 = Cast To Skyfire_AlienGlyph!(execute, Object=GetChildAt(self=Get AlienGlyph_HBox(), Index=@n1)) -> (then, CastFailed, AsSkyfire Alien Glyph:Skyfire Alien Glyph Object Reference) {is=title K2Node_DynamicCast}
   > Set bLongMessage!(
    execute
    bLongMessage:bool="integer > integer"(A=Get CurrMessageGlyphCount() -> (CurrMessageGlyphCount=0), B=12)
    self:Skyfire Alien Glyph Object Reference=@n2
   ) -> (then, Output_Get:bool)
   > PlayIntro(self:Skyfire Alien Glyph Object Reference=@n2) {is=title K2Node_CallFunction}
  ]}
 PlayOutro {nodes [
   PlayOutro() {is=title K2Node_FunctionEntry}
   > PlayAnimation(InAnimation=Get OutroAnim())
  ]}
 PlayReveal {nodes [
   PlayReveal() {is=title K2Node_FunctionEntry}
   > SetVisibility(self=Get RichText_Message(), InVisibility=HitTestInvisible)
   > PlayAnimation(InAnimation=Get RevealText_anim())
  ]}
 SetupInitialGlyphs {nodes [
   SetupInitialGlyphs() {is=title K2Node_FunctionEntry}
   n1 = Get AlienGlyph_HBox()
   > n2 = For Loop(LastIndex=int - int(A=GetChildrenCount(self=@n1) -> (ReturnValue=0), B=1))
   > n3 = Cast To Skyfire_AlienGlyph(Object=GetChildAt(self=Get AlienGlyph_HBox(), Index=@n2))
   > Set GlyphMaterial!(execute, GlyphMaterial:Material Interface Object Reference=Get AlienGlyphMaterial(), self:Skyfire Alien Glyph Object Reference=@n3) -> (then, Output_Get:Material Interface Object Reference)
   n4 = Get GlyphSize()
   > Set GlyphSize!(execute, GlyphSize:Vector 2D Structure=MakeVector2D(X=@n4, Y=@n4), self:Skyfire Alien Glyph Object Reference=@n3) -> (then, Output_Get:Vector 2D Structure)
   n5 = float / float(A=Get GlyphSize(), B=-7)
   n6 = Get GlyphSpacingAdjustment()
   > SetPadding(self=@n3, InPadding=Make Margin(Left=0, Top=0, Right=float + float(A=@n5, B=@n6), Bottom=0))
   > RandomizeGlyph(self:Skyfire Alien Glyph Object Reference=@n3)
  ]}
 TriggerGlyphWipe {nodes [
   TriggerGlyphWipe() {is=title K2Node_FunctionEntry}
   n1 = Get AlienGlyph_HBox()
   > n2 = For Loop(LastIndex=int - int(A=GetChildrenCount(self=@n1) -> (ReturnValue=0), B=1))
   > n3 = Cast To Skyfire_AlienGlyph(Object=GetChildAt(self=Get AlienGlyph_HBox(), Index=@n2))
   > PlayWipeAnim(self:Skyfire Alien Glyph Object Reference=@n3)
  ]}
 UpdateGlyphMaterial {nodes [
   n1 = UpdateGlyphMaterial!() -> (then, NewMaterial:Material Interface Object Reference) {is=title K2Node_FunctionEntry}
   > Set AlienGlyphMaterial(AlienGlyphMaterial:Material Interface Object Reference=@n1) -> (Output_Get:Material Interface Object Reference)
   n2 = Get AlienGlyph_HBox()
   > n3 = For Loop(LastIndex=int - int(A=GetChildrenCount(self=@n2) -> (ReturnValue=0), B=1))
   > n4 = Cast To Skyfire_AlienGlyph(Object=GetChildAt(self=Get AlienGlyph_HBox(), Index=@n3))
   > UpdateMaterial!(execute, self:Skyfire Alien Glyph Object Reference=@n4, NewMaterial:Material Interface Object Reference=Get AlienGlyphMaterial()) -> (then)
  ]}
 UpdateMessageDisplay {nodes [
   n1 = UpdateMessageDisplay!() -> (then, MessageText:Text) {is=title K2Node_FunctionEntry}
   > n2 = Branch(Condition=TextIsEmpty(InText=@n1))
   > SetVisibility(self=[Get RichText_Message(), Get BackingGlow_HBox(), Get AlienGlyph_HBox()], InVisibility=Collapsed)
   SetRenderOpacity(execute=@n2.else, self=Get RichText_Message(), InOpacity=0)
   > SetVisibility(self=Get RichText_Message(), InVisibility=HitTestInvisible)
   > SetText(self=Get RichText_Message(), InText=@n1)
   > n3 = Set Timer by Event(Delegate=Create Event(), Time=0.1, bLooping=false)
   > Set BuildGlyphTimer(BuildGlyphTimer:Timer Handle Structure=@n3) -> (Output_Get:Timer Handle Structure)
  ]}
]
variables [
 AlienGlyphMaterial {type=MaterialInterface, default=/Skyfire/UI/Art/M_UI_Skyfire_Glyphs_Inst}
 BuildGlyphTimer {type=TimerHandle}
 CurrMessageGlyphCount {type=integer}
 GlyphSize {type=number, default=44}
 GlyphSpacingAdjustment {type=number}
 OnIntroAnimFinished {}
 OnOutroAnimFinished {}
 PreviewText {type=string, default="<gameplaymessageprimary>[PH] UNKNOWN TRANSMISSION</>"}
 RevealTimerHandle {type=TimerHandle}
 TextRevealDelay {type=number, default=1}
 TextStyleSet {type=DataTable, default=&/SkyfireTextStyles}
 VOSound {type=SoundBase, default=/Skyfire/Sound/VO/Skyfire_VO_Backpack_Feedback_Positive_Cue}
]
widgets {
 info {inheritedWidgetCount=1, parentClass=/Script/SkyfireUI.SkyfireWidgetBase, rootWidgetClass=&Overlay, widgetCount=19}
 widgets [
  Overlay_17 {widgetClassPath=&Overlay, widget=Overlay_17}
  BackingGlow_HBox {
   widgetClassPath=&HorizontalBox
   bIsVariable=true
   parent=Overlay_17
   widget=BackingGlow_HBox
   properties {visibility=Collapsed, renderTransform {angle=0, scale=<1.6 1.4>, shear=<0 0>, translation=<0 0>}}
   slot {class=OverlaySlot, properties {horizontalAlignment=HAlign_Fill, verticalAlignment=VAlign_Fill}}
  }
  GlowBacking_Left {
   widgetClassPath=&Image
   bIsVariable=true
   parent=BackingGlow_HBox
   widget=GlowBacking_Left
   properties {
    brush {
     bIsDynamicallyLoaded=false
     drawAs=Image
     imageType=NoImage
     mirroring=NoMirror
     resourceName=None
     resourceObject=/Skyfire/UI/Art/M_UI_Skyfire_EnergyBacking_Text
     tiling=NoTile
     imageSize=<32 32>
     margin=<ltrb 0 0 0 0>
     outlineSettings {
      bUseBrushTransparency=false
      roundingType=HalfHeightRadius
      width=0
      color {colorUseRule=UseColor_Specified, specifiedColor=<rgba 0 0 0 0>}
      cornerRadii=<0 0 0 0>
     }
     tintColor {colorUseRule=UseColor_Specified, specifiedColor=<rgba 0.9281 0.1667 1 1>}
     uVRegion {bIsValid=false, max=<0 0>, min=<0 0>}
    }
    colorAndOpacity=<rgba 1 1 1 0.8>
    renderTransform {angle=0, scale=<-1 1>, shear=<0 0>, translation=<0 0>}
   }
   slot {class=HorizontalBoxSlot, properties {size {sizeRule=Fill, value=1}}}
  }
  GlowBacking_Right {
   widgetClassPath=&Image
   bIsVariable=true
   parent=BackingGlow_HBox
   widget=GlowBacking_Right
   properties {
    brush {
     bIsDynamicallyLoaded=false
     drawAs=Image
     imageType=NoImage
     mirroring=NoMirror
     resourceName=None
     resourceObject=&M_UI_Skyfire_EnergyBacking_Text
     tiling=NoTile
     imageSize=<32 32>
     margin=<ltrb 0 0 0 0>
     outlineSettings {
      bUseBrushTransparency=false
      roundingType=HalfHeightRadius
      width=0
      color {colorUseRule=UseColor_Specified, specifiedColor=<rgba 0 0 0 0>}
      cornerRadii=<0 0 0 0>
     }
     tintColor {colorUseRule=UseColor_Specified, specifiedColor=<rgba 0.9281 0.1667 1 1>}
     uVRegion {bIsValid=false, max=<0 0>, min=<0 0>}
    }
    colorAndOpacity=<rgba 1 1 1 0.8>
   }
   slot {class=HorizontalBoxSlot, properties {size {sizeRule=Fill, value=1}}}
  }
  AlienGlyph_HBox {
   widgetClassPath=&HorizontalBox
   bIsVariable=true
   parent=Overlay_17
   widget=AlienGlyph_HBox
   properties {renderTransform {angle=0, scale=<1 1>, shear=<-8 0>, translation=<0 -2>}}
   slot {class=OverlaySlot, properties {horizontalAlignment=HAlign_Center, verticalAlignment=VAlign_Center}}
  }
  Skyfire_AlienGlyph {widgetClassPath=&/Skyfire_AlienGlyph, parent=AlienGlyph_HBox, widget=Skyfire_AlienGlyph, slot {class=HorizontalBoxSlot}}
  Skyfire_AlienGlyph_690 {
   widgetClassPath=&Skyfire_AlienGlyph
   parent=AlienGlyph_HBox
   widget=Skyfire_AlienGlyph_690
   properties {glyphIndex=1}
   slot {class=HorizontalBoxSlot}
  }
  Skyfire_AlienGlyph_791 {
   widgetClassPath=&Skyfire_AlienGlyph
   parent=AlienGlyph_HBox
   widget=Skyfire_AlienGlyph_791
   properties {glyphIndex=2}
   slot {class=HorizontalBoxSlot}
  }
  Skyfire_AlienGlyph_887 {
   widgetClassPath=&Skyfire_AlienGlyph
   parent=AlienGlyph_HBox
   widget=Skyfire_AlienGlyph_887
   properties {glyphIndex=3}
   slot {class=HorizontalBoxSlot}
  }
  Skyfire_AlienGlyph_1004 {
   widgetClassPath=&Skyfire_AlienGlyph
   parent=AlienGlyph_HBox
   widget=Skyfire_AlienGlyph_1004
   properties {glyphIndex=4}
   slot {class=HorizontalBoxSlot}
  }
  Skyfire_AlienGlyph_1109 {
   widgetClassPath=&Skyfire_AlienGlyph
   parent=AlienGlyph_HBox
   widget=Skyfire_AlienGlyph_1109
   properties {glyphIndex=5}
   slot {class=HorizontalBoxSlot}
  }
  Skyfire_AlienGlyph_1195 {
   widgetClassPath=&Skyfire_AlienGlyph
   parent=AlienGlyph_HBox
   widget=Skyfire_AlienGlyph_1195
   properties {glyphIndex=6}
   slot {class=HorizontalBoxSlot}
  }
  Skyfire_AlienGlyph_1272 {
   widgetClassPath=&Skyfire_AlienGlyph
   parent=AlienGlyph_HBox
   widget=Skyfire_AlienGlyph_1272
   properties {glyphIndex=7}
   slot {class=HorizontalBoxSlot}
  }
  Skyfire_AlienGlyph_1352 {
   widgetClassPath=&Skyfire_AlienGlyph
   parent=AlienGlyph_HBox
   widget=Skyfire_AlienGlyph_1352
   properties {glyphIndex=8}
   slot {class=HorizontalBoxSlot}
  }
  Skyfire_AlienGlyph_1424 {
   widgetClassPath=&Skyfire_AlienGlyph
   parent=AlienGlyph_HBox
   widget=Skyfire_AlienGlyph_1424
   properties {glyphIndex=9}
   slot {class=HorizontalBoxSlot}
  }
  Skyfire_AlienGlyph_1487 {
   widgetClassPath=&Skyfire_AlienGlyph
   parent=AlienGlyph_HBox
   widget=Skyfire_AlienGlyph_1487
   properties {glyphIndex=10}
   slot {class=HorizontalBoxSlot}
  }
  Skyfire_AlienGlyph_1586 {
   widgetClassPath=&Skyfire_AlienGlyph
   parent=AlienGlyph_HBox
   widget=Skyfire_AlienGlyph_1586
   properties {glyphIndex=11}
   slot {class=HorizontalBoxSlot}
  }
  RichText_Message {
   widgetClassPath=/Script/UMG.RichTextBlock
   bIsVariable=true
   parent=Overlay_17
   widget=RichText_Message
   properties {
    justification=Center
    text="<gameplaymessageprimary>[PH] UNKNOWN TRANSMISSION</>"
    textStyleSet=&SkyfireTextStyles
    textTransformPolicy=ToUpper
    visibility=HitTestInvisible
    defaultTextStyleOverride {
     overflowPolicy=Clip
     transformPolicy=None
     colorAndOpacity {colorUseRule=UseColor_Specified, specifiedColor=<rgba 1 0 1 1>}
     font {
      bForceMonospaced=false
      bMaterialIsStencil=false
      fontMaterial=/Skyfire/UI/Art/M_UI_Skyfire_OpactitySwipe
      fontObject=&Burbank
      letterSpacing=0
      monospacedWidth=1
      size=40
      skewAmount=0
      typefaceFontName=Burbank-Black
      outlineSettings {
       bApplyOutlineToDropShadows=false
       bMiteredCorners=false
       bSeparateFillAlpha=false
       outlineMaterial=null
       outlineSize=0
       outlineColor=<rgba 0 0 0 1>
      }
     }
     highlightColor {colorUseRule=UseColor_Specified, specifiedColor=<rgba 0 0 0 1>}
     highlightShape {
      bIsDynamicallyLoaded=false
      drawAs=Image
      imageType=NoImage
      mirroring=NoMirror
      resourceName=None
      resourceObject=null
      tiling=NoTile
      imageSize=<32 32>
      margin=<ltrb 0 0 0 0>
      outlineSettings {
       bUseBrushTransparency=false
       roundingType=HalfHeightRadius
       width=0
       color {colorUseRule=UseColor_Specified, specifiedColor=<rgba 0 0 0 0>}
       cornerRadii=<0 0 0 0>
      }
      tintColor {colorUseRule=UseColor_Specified, specifiedColor=<rgba 1 1 1 1>}
      uVRegion {bIsValid=false, max=<0 0>, min=<0 0>}
     }
     selectedBackgroundColor {colorUseRule=UseColor_Foreground, specifiedColor=<rgba 1 0 1 1>}
     shadowColorAndOpacity=<rgba 0 0 0 1>
     shadowOffset=<0 0>
     strikeBrush {
      bIsDynamicallyLoaded=false
      drawAs=Image
      imageType=NoImage
      mirroring=NoMirror
      resourceName=None
      resourceObject=null
      tiling=NoTile
      imageSize=<32 32>
      margin=<ltrb 0 0 0 0>
      outlineSettings {
       bUseBrushTransparency=false
       roundingType=HalfHeightRadius
       width=0
       color {colorUseRule=UseColor_Specified, specifiedColor=<rgba 0 0 0 0>}
       cornerRadii=<0 0 0 0>
      }
      tintColor {colorUseRule=UseColor_Specified, specifiedColor=<rgba 1 1 1 1>}
      uVRegion {bIsValid=false, max=<0 0>, min=<0 0>}
     }
     underlineBrush {
      bIsDynamicallyLoaded=false
      drawAs=Image
      imageType=NoImage
      mirroring=NoMirror
      resourceName=None
      resourceObject=null
      tiling=NoTile
      imageSize=<32 32>
      margin=<ltrb 0 0 0 0>
      outlineSettings {
       bUseBrushTransparency=false
       roundingType=HalfHeightRadius
       width=0
       color {colorUseRule=UseColor_Specified, specifiedColor=<rgba 0 0 0 0>}
       cornerRadii=<0 0 0 0>
      }
      tintColor {colorUseRule=UseColor_Specified, specifiedColor=<rgba 1 1 1 1>}
      uVRegion {bIsValid=false, max=<0 0>, min=<0 0>}
     }
    }
    renderTransform {angle=0, scale=<1 1>, shear=<-8 0>, translation=<0 0>}
   }
   slot {class=OverlaySlot, properties {horizontalAlignment=HAlign_Center}}
  }
  TutorialHighlight {widgetClassPath=&TutorialHighlight}
 ]
}
