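# Skyfire HUD widget that shows mothership gameplay messages (experiment start, orb spawns, loot tier, vault entry) and the final vault countdown.
# Player pawn component, mutator phase and timer events request messages, whose text is overridden by type and rarity.
# Messages display immediately or enter a priority queue, with intro/outro animations, icons and tier indicators.
# Near countdown's end, optional rarity-matched performance flavor text is picked randomly from a data table.
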WidgetBlueprint /Skyfire/UI/Skyfire_Messaging_Gameplay
animations [
 IntroAnimation {
  binding_count=2
  length_seconds=1.25
  object=&self:IntroAnimation
  sequence {bindings [
    SizeBox_TierIndicator {
     id=#1
     possessed_class=SizeBox
     tracks [{
       type=MovieScene2DTransformTrack
       sections [{
         type=MovieScene2DTransformSection
         end_frame=25
         start_frame=0
         channels [
          Translation.X {type=MovieSceneScriptingFloatChannel, const [0,2]}
          Translation.Y {type=MovieSceneScriptingFloatChannel, const [0,2]}
          Angle {type=MovieSceneScriptingFloatChannel, const [0,2]}
          Scale.X {type=MovieSceneScriptingFloatChannel, keys [[0,0.5],[25,1]]}
          Scale.Y {type=MovieSceneScriptingFloatChannel, keys [[0,0],[25,1]]}
          Shear.X {type=MovieSceneScriptingFloatChannel, const [0,2]}
          Shear.Y {type=MovieSceneScriptingFloatChannel, const [0,2]}
         ]
        }]
      }]
    }
    Overlay_Icons {
     id=#2
     possessed_class=Overlay
     tracks [
      {
       type=MovieSceneFloatTrack
       display_name=Render Opacity
       sections [{
         type=MovieSceneFloatSection
         end_frame=25
         start_frame=0
         channels [None {type=MovieSceneScriptingFloatChannel, keys [[0,0],[7,0],[15,1],[25,1]]}]
        }]
      }
      {
       type=MovieScene2DTransformTrack
       sections [{
         type=MovieScene2DTransformSection
         end_frame=25
         start_frame=0
         channels [
          Translation.X {type=MovieSceneScriptingFloatChannel, const [0,5]}
          Translation.Y {type=MovieSceneScriptingFloatChannel, const [0,5]}
          Angle {type=MovieSceneScriptingFloatChannel, const [0,5]}
          Scale.X {type=MovieSceneScriptingFloatChannel, keys [[0,0.5],[7,0.5],[14,1.5],[15,1],[25,1]]}
          Scale.Y {type=MovieSceneScriptingFloatChannel, keys [[0,0.5],[7,0.5],[14,1.5],[15,1],[25,1]]}
          Shear.X {type=MovieSceneScriptingFloatChannel, const [0,5]}
          Shear.Y {type=MovieSceneScriptingFloatChannel, const [0,5]}
         ]
        }]
      }
     ]
    }
   ]}
 }
 OutroAnimation {
  binding_count=2
  length_seconds=1.25
  object=&self:OutroAnimation
  sequence {bindings [
    SizeBox_TierIndicator {
     id=#3
     possessed_class=SizeBox
     tracks [{
       type=MovieScene2DTransformTrack
       sections [{
         type=MovieScene2DTransformSection
         end_frame=25
         start_frame=0
         channels [
          Translation.X {type=MovieSceneScriptingFloatChannel, const [0,3]}
          Translation.Y {type=MovieSceneScriptingFloatChannel, const [0,3]}
          Angle {type=MovieSceneScriptingFloatChannel, const [0,3]}
          Scale.X {type=MovieSceneScriptingFloatChannel, keys [[0,1],[15,0.5],[25,0.5]]}
          Scale.Y {type=MovieSceneScriptingFloatChannel, keys [[0,1],[15,0],[25,0]]}
          Shear.X {type=MovieSceneScriptingFloatChannel, const [0,3]}
          Shear.Y {type=MovieSceneScriptingFloatChannel, const [0,3]}
         ]
        }]
      }]
    }
    Overlay_Icons {
     id=#4
     possessed_class=Overlay
     tracks [
      {
       type=MovieSceneFloatTrack
       display_name=Render Opacity
       sections [{type=MovieSceneFloatSection, end_frame=25, start_frame=0, channels [None {type=MovieSceneScriptingFloatChannel, keys [[0,1],[15,0],[25,0]]}]}]
      }
      {
       type=MovieScene2DTransformTrack
       sections [{
         type=MovieScene2DTransformSection
         end_frame=25
         start_frame=0
         channels [
          Translation.X {type=MovieSceneScriptingFloatChannel, const [0,2]}
          Translation.Y {type=MovieSceneScriptingFloatChannel, const [0,2]}
          Angle {type=MovieSceneScriptingFloatChannel, const [0,2]}
          Scale.X {type=MovieSceneScriptingFloatChannel, const [1,2]}
          Scale.Y {type=MovieSceneScriptingFloatChannel, const [1,2]}
          Shear.X {type=MovieSceneScriptingFloatChannel, const [0,2]}
          Shear.Y {type=MovieSceneScriptingFloatChannel, const [0,2]}
         ]
        }]
      }
     ]
    }
   ]}
 }
]
graphs [
 EnableSkyfireMessages {nodes [
   n1 = EnableSkyfireMessages!() -> (then, InSkyfirePlayerComponent:Skyfire Player Pawn Component Object Reference) {is=title K2Node_FunctionEntry}
   > Is Valid(InputObject=@n1)
   > SetVisibility(InVisibility=Select(
     Option 0=Collapsed
     Option 1=SelfHitTestInvisible
     Index=IsInMothership!(self:Skyfire Player Pawn Component Object Reference=@n1) -> (ReturnValue:bool)
    ))
  ]}
 EvaluateConditionsToStartFinalTimer {nodes [
   EvaluateConditionsToStartFinalTimer() {is=title K2Node_FunctionEntry}
   > Branch(Condition=Get IsFinalVaultTimerStarted() -> (IsFinalVaultTimerStarted:bool))
   > EnableFinalVaultTimer()
  ]}
 EventGraph {nodes [
   OnPlayerEnteredLootRoom!() -> (OutputDelegate, then, SkyfirePawnComponent:Skyfire Player Pawn Component Object Reference) {is=title K2Node_CustomEvent}
   > ClearGameplayMessaging!(execute, self, AllowsAdditionalMessaging:bool=true) -> (then)
   > Delay(Duration=Get DelaySkyfireEnterLootVaultMessageTime() -> (DelaySkyfireEnterLootVaultMessageTime:double))
   > InitNewMessage!(execute, self, CurrentMessage:SkyfireGameplayMessageTypeEnum Enum=NewEnumerator5) -> (then)
   OnExitInitialPrisonTimeChanged!() -> (OutputDelegate, then, NewTime:float=~) {is=title K2Node_CustomEvent}
   > EnableSkyfireMessages!(
    execute
    self
    InSkyfirePlayerComponent:Skyfire Player Pawn Component Object Reference=Get CachedSkyfirePlayerPawnComponent() -> (CachedSkyfirePlayerPawnComponent:Skyfire Player Pawn Component Object Reference)
   ) -> (then) {is=title K2Node_CallFunction}
   > Delay(Duration=0.2)
   > InitNewMessage(CurrentMessage=NewEnumerator1)
   n1 = DisplayNewMessage!() -> (OutputDelegate, then, OverwrittenDisplayData:Skyfire Gameplay Message Data Structure) {is=title K2Node_CustomEvent}
   > UpdateMessageDisplayFromData!(execute, self, Display Data:Skyfire Gameplay Message Data Structure=@n1.OverwrittenDisplayData) -> (then)
   > n2 = "Play Animation with Finished event (legacy)"!(
    execute:Exec
    Widget:User Widget Object Reference=Self-Reference()
    InAnimation:Widget Animation Object Reference=Get IntroAnimation() -> (IntroAnimation:Widget Animation Object Reference)
    StartAtTime:float=0
    NumLoopsToPlay:int=1
    PlayMode:EUMGSequencePlayMode Enum
    PlaybackSpeed:float=1
   ) -> (then:Exec, Result:UMGSequence Player Object Reference, Finished:Exec) {is=title K2Node_PlayAnimation}
   n3 = Set Timer by Event(
    execute=@n2.Finished
    Delegate=Create Event()
    Time=Break Skyfire Gameplay Message Data!(SkyfireGameplayMessageData:"Skyfire Gameplay Message Data Structure (by ref)"=@n1.OverwrittenDisplayData) -> (
     PrimaryMessageText_2_#5:Text
     SecondaryMessageText_4_#6:Text
     IncludeStarsTierIndicator_27_#7:bool=~
     DisplayIcon_7_#8:Texture 2D Object Reference
     IconAddShadow_13_#9:bool=~
     QueuePriority_19_#10:int
     MessageDisplayTime_24_#11:float=~
    ) {is=title K2Node_BreakStruct}.MessageDisplayTime_24_#11
    bLooping=false
   )
   > Set TimerHandleMessageDisplayTime(TimerHandleMessageDisplayTime:Timer Handle Structure=@n3) -> (Output_Get:Timer Handle Structure)
   n4 = OnPawnComponentAttached!() -> (OutputDelegate, then, OwnerPawn:Fort Pawn Object Reference, AttachedComponent:Fort Pawn Component Object Reference) {is=title K2Node_CustomEvent}
   > n5 = Cast To SkyfirePlayerPawnComponent!(execute, Object=@n4.AttachedComponent) -> (then, CastFailed, AsSkyfire Player Pawn Component:Skyfire Player Pawn Component Object Reference) {is=title K2Node_DynamicCast}
   n6 = GetOwningPlayerPawn!(self:User Widget Object Reference) -> (ReturnValue:Pawn Object Reference)
   n7 = Cast To FortPawn!(execute=[@n5.then, @n11.Is Valid], Object=@n6) -> (then, CastFailed, AsFort Pawn:Fort Pawn Object Reference) {is=title K2Node_DynamicCast}
   n8 = GetComponentByClass!(self:Actor Object Reference=@n7, ComponentClass:Actor Component Class Reference=/Script/SkyfireRuntime.SkyfirePlayerPawnComponent) -> (ReturnValue:Skyfire Player Pawn Component Object Reference)
   > n9 = Is Valid(InputObject=@n8)
   Unbind Event from On Pawn Component Attached Event(execute=@n9.Is Not Valid, self:Fort Pawn Object Reference=@n7, Delegate=@n4) {is=title K2Node_RemoveDelegate}
   > Bind Event to On Pawn Component Attached Event(self:Fort Pawn Object Reference=@n7, Delegate=@n4) {is=title K2Node_AddDelegate}
   UnbindPlayerPawn(execute=@n9.Is Valid)
   > Set Cached Skyfire Player Pawn Component!(execute, self, InSkyfirePlayerPawnComponent:Skyfire Player Pawn Component Object Reference=@n8) -> (then) {is=title K2Node_CallFunction}
   n10 = OnPawnChanged!() -> (OutputDelegate, then, NewPawn:Fort Pawn Object Reference) {is=title K2Node_CustomEvent}
   n11 = Is Valid(exec=[@n10.then, @n53.then], InputObject=@n6)
   EnableFinalVaultTimer() {is=title K2Node_CustomEvent}
   > Get!(execute:Exec) -> (then:Exec, else:Exec, CachedSkyfirePlayerPawnComponent:Skyfire Player Pawn Component Object Reference)
   > UpdateFinalVaultCountdownTimer()
   > SetVisibility(
    self=Get CommonText_Clock_VaultTimeRemaining() -> (CommonText_Clock_VaultTimeRemaining:Common Numeric Text Block Object Reference)
    InVisibility=SelfHitTestInvisible
   )
   > n12 = Set bHasActiveContent(bHasActiveContent:bool=true) -> (Output_Get:bool)
   > SetHasActiveContent!(execute, self:Fort Prioritized Widget Interface, bNewActive:bool=@n12) -> (then)
   UpdateFinalVaultCountdownTimer() {is=title K2Node_CustomEvent}
   > n13 = Get()
   n14 = GetSecondsUntil!(
    self:Skyfire Player Pawn Component Object Reference=@n13
    Time:float=GetExitLootRoomTime!(self:Skyfire Player Pawn Component Object Reference=@n13) -> (ReturnValue:float)
   ) -> (ReturnValue:float)
   > SetCurrentValue!(execute, self:Common Numeric Text Block Object Reference=Get CommonText_Clock_VaultTimeRemaining(), NewValue:float=@n14) -> (then)
   n15 = Get TimeUntilVaultWarningColorMaxThreshold() -> (TimeUntilVaultWarningColorMaxThreshold:int=0)
   n16 = Ceil(A=@n14) {is=title K2Node_CallFunction}
   > SetColorAndOpacity(
    self=Get CommonText_Clock_VaultTimeRemaining()
    InColorAndOpacity=Select(
     Option 0:Slate Color Structure=Get ColorVaultTimerWarningThreshold() -> (ColorVaultTimerWarningThreshold:Slate Color Structure)
     Option 1:Slate Color Structure=Get ColorVaultTimerDefault() -> (ColorVaultTimerDefault:Slate Color Structure)
     Index="integer > integer"(A=@n16, B=@n15) {is=title K2Node_CallFunction}
    ) -> (ReturnValue:Slate Color Structure)
   )
   n17 = Get FlavorTextDisplayThreshold() -> (FlavorTextDisplayThreshold:int=0)
   n18 = Get bStartedFlavorTextDisplay() -> (bStartedFlavorTextDisplay:bool)
   n19 = Get CachedCurrentLootRarity() -> (CachedCurrentLootRarity:EFortRarity Enum=Common)
   > n20 = Branch(Condition=AND Boolean!(
     A="integer <= integer"!(A:int=@n16, B:int=@n17) -> (ReturnValue:bool) {is=title K2Node_CallFunction}
     B=NOT Boolean(A=@n18)
     C:bool="Equal (Enum)"(A:EFortRarity Enum=@n19, B:EFortRarity Enum=Legendary)
    ) -> (ReturnValue))
   > n21 = TriggerFlavorText()
   n22 = Branch(execute=[@n21.then, @n20.else], Condition="integer <= integer"(A=@n16, B=0))
   > n23 = Branch(Condition=Get bAllowedToShowFlavorText() -> (bAllowedToShowFlavorText:bool))
   EndSkyfireMesssaging(execute=@n23.else)
   n24 = Set Timer by Event(execute=@n22.else, Delegate=Create Event(), Time=Get FinalVaultTimerUpdateRate() -> (FinalVaultTimerUpdateRate:double), bLooping=false)
   > Set TimerHandleFinalVaultCountdown(TimerHandleFinalVaultCountdown:Timer Handle Structure=@n24) -> (Output_Get:Timer Handle Structure)
   n25 = SkyfirePhaseChanged!() -> (OutputDelegate, then, NewPhase:ESkyfirePhase Enum=None) {is=title K2Node_CustomEvent}
   > n26 = Switch on ESkyfirePhase!(execute, Selection:ESkyfirePhase Enum=@n25.NewPhase) -> (None:Exec, Prison:Exec, CoreGameplay:Exec, AllPlayersExited:Exec) {is=title K2Node_SwitchEnum}
   InitOrbsAddedTimers(execute=@n26.CoreGameplay)
   > n27 = Get!(execute) -> (then, else, CachedSkyfireMutator:Fort Athena Mutator Skyfire Object Reference)
   > Unbind Event from On Skyfire Phase Changed(self:Fort Athena Mutator Skyfire Object Reference=@n27, Delegate=@n25) {is=title K2Node_RemoveDelegate}
   n28 = ClearGameplayMessaging!() -> (OutputDelegate, then, AllowsAdditionalMessaging:bool=~) {is=title K2Node_CustomEvent}
   > Set ReadyForNewMessage(ReadyForNewMessage:bool=@n28.AllowsAdditionalMessaging) -> (Output_Get:bool)
   > Clear(TargetArray:Array of Skyfire Gameplay Message Data Structures=Get GameplayMessageQueue() -> (GameplayMessageQueue:Array of Skyfire Gameplay Message Data Structures))
   > UpdateGlyphMaterial!(
    execute
    self:Skyfire Alien Message Widget Object Reference=Get PrimaryMessageWidget() -> (PrimaryMessageWidget:Skyfire Alien Message Widget Object Reference)
    NewMaterial:Material Interface Object Reference=/Skyfire/UI/Art/M_UI_Skyfire_Glyphs_Inst
   ) -> (then)
   > n29 = Set bHasActiveContent(bHasActiveContent=Get IsFinalVaultTimerStarted())
   > SetHasActiveContent(bNewActive=@n29)
   > SetVisibility(
    self=Get VerticalBox_GameplayMessagingLayout() -> (VerticalBox_GameplayMessagingLayout:Vertical Box Object Reference)
    InVisibility=Collapsed
   )
   > Clear and Invalidate Timer by Handle(Handle=Get TimerHandleMessageDisplayTime() -> (TimerHandleMessageDisplayTime:Timer Handle Structure))
   OnTimeOrbsSpawningActivated() {is=title K2Node_CustomEvent}
   > InitNewMessage(CurrentMessage=NewEnumerator2)
   > Clear and Invalidate Timer by Handle(Handle=Get TimerHandleTimeOrbsSpawnDelay() -> (TimerHandleTimeOrbsSpawnDelay:Timer Handle Structure))
   InitOrbsAddedTimers() {is=title K2Node_CustomEvent}
   > n30 = Set Timer by Event(
    Delegate=Create Event()
    Time=GetValueAtLevel!(
     Input:"Scalable Float Structure (by ref)"=Get RefTimeOrbIntialSpawnDelay() -> (RefTimeOrbIntialSpawnDelay:Scalable Float Structure)
     Level:float=0
    ) -> (ReturnValue:float)
    bLooping=false
    InitialStartDelay=GetValueAtLevel(Input=Get RefTimeToLoadFromPrisonToGamespace() -> (RefTimeToLoadFromPrisonToGamespace:Scalable Float Structure), Level=0)
   )
   > Set TimerHandleTimeOrbsSpawnDelay(TimerHandleTimeOrbsSpawnDelay:Timer Handle Structure=@n30) -> (Output_Get:Timer Handle Structure)
   > n31 = Set Timer by Event(
    Delegate=Create Event()
    Time=GetValueAtLevel(Input=Get RefVaultOrbIntialSpawnDelay() -> (RefVaultOrbIntialSpawnDelay:Scalable Float Structure), Level=0)
    bLooping=false
   )
   > Set TimerHandleVaultOrbsSpawnDelay(TimerHandleVaultOrbsSpawnDelay:Timer Handle Structure=@n31) -> (Output_Get:Timer Handle Structure)
   OnVaultOrbsSpawningActivated() {is=title K2Node_CustomEvent}
   > InitNewMessage(CurrentMessage=NewEnumerator3)
   > Clear and Invalidate Timer by Handle(Handle=Get TimerHandleVaultOrbsSpawnDelay() -> (TimerHandleVaultOrbsSpawnDelay:Timer Handle Structure))
   > Branch(Condition=Get bAllowedToShowFlavorText())
   > Set bTimerToLootRoomSet(bTimerToLootRoomSet:bool=true) -> (Output_Get:bool)
   > n32 = Set Timer by Function Name!(
    execute
    Object:Object Reference
    FunctionName:String=TimerToLootRoom
    Time:float=0.1
    bLooping:bool=true
    bMaxOncePerFrame:bool=false
    InitialStartDelay:float=0
    InitialStartDelayVariance:float=0
   ) -> (then, ReturnValue:Timer Handle Structure) {is=title K2Node_CallFunction}
   > Set TimeUntilLootRoom_Timer(TimeUntilLootRoom_Timer:Timer Handle Structure=@n32) -> (Output_Get:Timer Handle Structure)
   TriggerFlavorText() {is=title K2Node_CustomEvent}
   > Branch(Condition=Get bAllowedToShowFlavorText())
   > n33 = GetFlavorText!(execute, self, inrarity:EFortRarity Enum=Get CachedCurrentLootRarity() -> (CachedCurrentLootRarity=Common)) -> (then, PerformanceMessage:Text, Success:bool)
   > n34 = Branch(Condition=@n33.Success)
   > Set PerfText(PerfText:Text=@n33) -> (Output_Get:Text)
   > Set bStartedFlavorTextDisplay(bStartedFlavorTextDisplay:bool=true) -> (Output_Get:bool)
   > Branch(Condition=Get bShowPerfFlavorText() -> (bShowPerfFlavorText:bool))
   > InitNewMessage(CurrentMessage=NewEnumerator7)
   PrintString!(execute=@n34.else, InString=Failed, bPrintToScreen, bPrintToLog, TextColor, Duration, Key) -> (then)
   > Set bIsFinalMessage(bIsFinalMessage:bool=true) -> (Output_Get:bool)
   n35 = OnInMotherShipChanged!() -> (OutputDelegate, then, bInMothership:bool=~) {is=title K2Node_CustomEvent}
   > n36 = Branch(Condition=@n35.bInMothership)
   Branch(execute=@n36.else, Condition=Get bStartedFlavorTextDisplay())
   > EndSkyfireMesssaging()
   n37 = UpdateMessageDisplayFromData!() -> (OutputDelegate, then, Display Data:Skyfire Gameplay Message Data Structure) {is=title K2Node_CustomEvent}
   > n38 = Sequence!(execute) -> (then_0, then_1, then_2:Exec, then_3:Exec)
   n39 = Break Skyfire Gameplay Message Data(SkyfireGameplayMessageData=@n37.Display Data) -> (
    IncludeStarsTierIndicator_27_#7=~
    IconAddShadow_13_#9=~
    MessageDisplayTime_24_#11=~
   )
   Handle Skyfire Gameplay Message!(
    ExecIn:Exec=@n38.then_1
    AlienWidget:Skyfire Alien Message Widget Object Reference=Get SecondaryMessageWidget() -> (SecondaryMessageWidget:Skyfire Alien Message Widget Object Reference)
    MessageText:Text=@n39.SecondaryMessageText_4_#6
   ) -> (ExecOut:Exec) {is=title K2Node_MacroInstance}
   n40 = Is Valid(exec=@n38.then_2, InputObject=@n39.DisplayIcon_7_#8)
   n41 = Get Image_Icon_shadow() -> (Image_Icon_shadow:Image Object Reference)
   n42 = Get Image_Icon() -> (Image_Icon:Image Object Reference)
   > SetBrushFromTexture!(
    execute
    self:Image Object Reference=[@n41, @n42]
    Texture:Texture 2D Object Reference=@n39.DisplayIcon_7_#8
    bMatchSize:bool=false
   ) -> (then)
   > SetVisibility(self=@n42, InVisibility=SelfHitTestInvisible)
   > SetVisibility(
    self=@n41
    InVisibility=Select(
     Option 0:ESlateVisibility Enum=Collapsed
     Option 1:ESlateVisibility Enum=SelfHitTestInvisible
     Index=@n39.IconAddShadow_13_#9
    ) -> (ReturnValue:ESlateVisibility Enum)
   )
   n43 = Get IconGlow_HBox() -> (IconGlow_HBox:Horizontal Box Object Reference)
   > SetVisibility(self=@n43, InVisibility=SelfHitTestInvisible)
   SetVisibility(execute=@n40.Is Not Valid, self=[@n43, @n41, @n42], InVisibility=Collapsed)
   n44 = Branch(execute=@n38.then_3, Condition=@n39.IncludeStarsTierIndicator_27_#7)
   n45 = Get CachedCurrentLootRarity() -> (CachedCurrentLootRarity=Common)
   > n46 = GetRarityData!(execute, Rarity:EFortRarity Enum=@n45) -> (then, ReturnValue:Fort Rarity Item Data Structure)
   n47 = Find(
    TargetMap:Map of EFortRarity Enums to Integers=Get RarityToOrbsCollectedMap() -> (RarityToOrbsCollectedMap:Map of EFortRarity Enums to Integers=Common)
    Key:"EFortRarity Enum (by ref)"=@n45
   ) -> (Value:int)
   > n48 = Branch(Condition=@n47.ReturnValue)
   n49 = Get SizeBox_TierIndicator() -> (SizeBox_TierIndicator:Size Box Object Reference)
   > SetVisibility(self=@n49, InVisibility=SelfHitTestInvisible)
   > n50 = For Each Loop(Array:Array of Skyfire HUDElement Tier Rating Object References=Get TierIndicatorWidgets() -> (TierIndicatorWidgets:Array of Skyfire HUDElement Tier Rating Object References)) -> (Array Element:Skyfire HUDElement Tier Rating Object Reference)
   n51 = "integer <= integer"(A=@n50.Array Index, B=@n47)
   n52 = Break Fort Rarity Item Data!(FortRarityItemData:"Fort Rarity Item Data Structure (by ref)"=@n46) -> (
    Name:Text
    Color1:Linear Color Structure
    Color2:Linear Color Structure
    Color3:Linear Color Structure
    Color4:Linear Color Structure
    Color5:Linear Color Structure
    Radius:float=~
    Falloff:float=~
    Brightness:float=~
    Roughness:float=~
    Glow:float=~
   ) {is=title K2Node_BreakStruct}
   > SetVisualsToTierLevel!(
    execute
    self:Skyfire HUDElement Tier Rating Object Reference=@n50
    FillColor:Linear Color Structure=SelectColor(A=@n52.Color1, B=Get Color_TierUnobtainedFill() -> (Color_TierUnobtainedFill:Linear Color Structure), bPickA=@n51)
    OutlineColor:Linear Color Structure=@n52.Color1
    GlowColor:Linear Color Structure=@n52.Color1
   ) -> (then)
   SetVisibility(execute=[@n48.else, @n44.else], self=@n49, InVisibility=Collapsed)
   Handle Skyfire Gameplay Message(ExecIn=@n38.then_0, AlienWidget=Get PrimaryMessageWidget(), MessageText=@n39)
   n53 = BindToPawnComponent() {is=title K2Node_CustomEvent}
   n54 = InitNewMessage!() -> (OutputDelegate, then, CurrentMessage:SkyfireGameplayMessageTypeEnum Enum=NewEnumerator0) {is=title K2Node_CustomEvent}
   > Get!(execute) -> (then, else, CachedSkyfirePlayerPawnComponent:Skyfire Player Pawn Component Object Reference)
   > n55 = Branch(Condition=Get ReadyForNewMessage() -> (ReadyForNewMessage:bool))
   > Set ReadyForNewMessage(ReadyForNewMessage=false)
   > n56 = Set bHasActiveContent(bHasActiveContent=true)
   > SetHasActiveContent(bNewActive=@n56)
   > SetVisibility(self=Get VerticalBox_GameplayMessagingLayout(), InVisibility=SelfHitTestInvisible)
   n57 = Get CachedCurrentLootRarity() -> (CachedCurrentLootRarity=Common)
   > n58 = OverrideMessageData!(execute, self, Current Message:SkyfireGameplayMessageTypeEnum Enum=@n54.CurrentMessage, TierRarity:EFortRarity Enum=@n57) -> (then, OverrideMessageData:Skyfire Gameplay Message Data Structure)
   > n59 = Branch(Condition=OR Boolean(
     A="Equal (Enum)"(A:SkyfireGameplayMessageTypeEnum Enum=@n54.CurrentMessage, B:SkyfireGameplayMessageTypeEnum Enum=NewEnumerator7)
     B="Equal (Enum)"(A=@n54.CurrentMessage, B=NewEnumerator8)
    ))
   > n60 = UpdateGlyphMaterial(self=Get PrimaryMessageWidget(), NewMaterial=/Skyfire/UI/Art/M_UI_Skyfire_Glyphs_Emphasis)
   DisplayNewMessage!(execute=[@n60.then, @n59.else], self, OverwrittenDisplayData:Skyfire Gameplay Message Data Structure=@n58) -> (then) {is=title K2Node_CallFunction}
   n61 = OverrideMessageData(execute=@n55.else, Current Message=@n54.CurrentMessage, TierRarity=@n57)
   n62 = Break Skyfire Gameplay Message Data(SkyfireGameplayMessageData=@n61) -> (
    IncludeStarsTierIndicator_27_#7=~
    IconAddShadow_13_#9=~
    MessageDisplayTime_24_#11=~
   )
   > n63 = Branch(Condition="Equal (Integer)"(A=@n62.QueuePriority_19_#10, B=0) {is=title K2Node_CallFunction})
   n64 = Get GameplayMessageQueue()
   > n65 = For Each Loop with Break!(Exec:Exec, Array:Array of Skyfire Gameplay Message Data Structures=@n64, Break:Exec=@n66.then) -> (LoopBody:Exec, Array Element:Skyfire Gameplay Message Data Structure, Array Index:int, Completed:Exec)
   > Branch(Condition="integer > integer"(
     A=Break Skyfire Gameplay Message Data(SkyfireGameplayMessageData=@n65) -> (
      IncludeStarsTierIndicator_27_#7=~
      IconAddShadow_13_#9=~
      MessageDisplayTime_24_#11=~
     ).QueuePriority_19_#10
     B=@n62.QueuePriority_19_#10
    ))
   > n66 = Insert!(
    execute
    TargetArray:Array of Skyfire Gameplay Message Data Structures=@n64
    NewItem:"Skyfire Gameplay Message Data Structure (by ref)"=@n61
    Index:int=@n65.Array Index
   ) -> (then) {is=title K2Node_CallArrayFunction}
   Add Unique(
    execute=[@n65.Completed, @n63.else]
    TargetArray:Array of Skyfire Gameplay Message Data Structures=@n64
    NewItem:"Skyfire Gameplay Message Data Structure (by ref)"=@n61
   )
   n67 = OnEarnedLootRarityChanged!() -> (OutputDelegate, then, NewRarity:EFortRarity Enum=Common) {is=title K2Node_CustomEvent}
   > Set CachedCurrentLootRarity(CachedCurrentLootRarity:EFortRarity Enum=@n67.NewRarity) -> (Output_Get:EFortRarity Enum=Common)
   > InitNewMessage(CurrentMessage=NewEnumerator4) {is=title K2Node_CallFunction}
   n68 = OnExitLootRoomTimeSet!() -> (OutputDelegate, then, NewTime:float=~) {is=title K2Node_CustomEvent}
   > Set IsFinalVaultTimerStarted(IsFinalVaultTimerStarted:bool=true) -> (Output_Get:bool)
   > EvaluateConditionsToStartFinalTimer() {is=title K2Node_CallFunction}
   > n69 = Get()
   > Unbind Event from On Exit Loot Room Time Changed(self:Skyfire Player Pawn Component Object Reference=@n69, Delegate=@n68) {is=title K2Node_RemoveDelegate}
   EvaluateMessageQueue() {is=title K2Node_CustomEvent}
   n70 = Get GameplayMessageQueue()
   > n71 = Branch(Condition=Is Array Non Empty!(InputArray:Array of Skyfire Gameplay Message Data Structures=@n70) -> (Non-Empty:bool=~) {is=title K2Node_MacroInstance})
   > Set ReadyForNewMessage(ReadyForNewMessage=false)
   > DisplayNewMessage(OverwrittenDisplayData="Get (a copy)"(Array:Array of Skyfire Gameplay Message Data Structures=@n70) -> (Output:Skyfire Gameplay Message Data Structure))
   > Remove Index!(execute, TargetArray:Array of Skyfire Gameplay Message Data Structures=@n70, IndexToRemove:int=0) -> (then)
   ClearGameplayMessaging(execute=@n71.else, AllowsAdditionalMessaging=true) {is=title K2Node_CallFunction}
   EndCurrentMessage() {is=title K2Node_CustomEvent}
   > PlayOutro(self:Skyfire Alien Message Widget Object Reference=[Get PrimaryMessageWidget(), Get SecondaryMessageWidget()])
   > n72 = "Play Animation with Finished event (legacy)"(
    Widget=Self-Reference()
    InAnimation=Get OutroAnimation() -> (OutroAnimation:Widget Animation Object Reference)
    StartAtTime=0
    NumLoopsToPlay=1
    PlaybackSpeed=1
   )
   n73 = Branch(execute=@n72.Finished, Condition=Get bIsFinalMessage() -> (bIsFinalMessage:bool))
   > Set bAreGameplayMessagesComplete(bAreGameplayMessagesComplete:bool=true) -> (Output_Get:bool)
   > ClearGameplayMessaging(AllowsAdditionalMessaging=false)
   Set ReadyForNewMessage(execute=@n73.else, ReadyForNewMessage=true)
   > EvaluateMessageQueue() {is=title K2Node_CallFunction}
   Event Construct()
   > SetVisibility(InVisibility=Collapsed)
   > Set TierIndicatorWidgets(TierIndicatorWidgets:Array of Skyfire HUDElement Tier Rating Object References=Make Array!(
     "[0]":Skyfire HUDElement Tier Rating Object Reference=Get Skyfire_HUDElement_TierRating_1() -> (Skyfire_HUDElement_TierRating_1:Skyfire HUDElement Tier Rating Object Reference)
     "[1]":Skyfire HUDElement Tier Rating Object Reference=Get Skyfire_HUDElement_TierRating_2() -> (Skyfire_HUDElement_TierRating_2:Skyfire HUDElement Tier Rating Object Reference)
     "[2]":Skyfire HUDElement Tier Rating Object Reference=Get Skyfire_HUDElement_TierRating_3() -> (Skyfire_HUDElement_TierRating_3:Skyfire HUDElement Tier Rating Object Reference)
     "[3]":Skyfire HUDElement Tier Rating Object Reference=Get Skyfire_HUDElement_TierRating_4() -> (Skyfire_HUDElement_TierRating_4:Skyfire HUDElement Tier Rating Object Reference)
     "[4]":Skyfire HUDElement Tier Rating Object Reference=Get Skyfire_HUDElement_TierRating_5() -> (Skyfire_HUDElement_TierRating_5:Skyfire HUDElement Tier Rating Object Reference)
    ) -> (Array:Array of Skyfire HUDElement Tier Rating Object References)) -> (Output_Get:Array of Skyfire HUDElement Tier Rating Object References)
   K2Node_Event_1 = EndSkyfireMesssaging() {is=title K2Node_CustomEvent}
   > n74 = Sequence!(execute) -> (then_0, then_1, then_2, then_3, then_4:Exec)
   > n75 = Get()
   > Unbind Event from On Earned Loot Rarity Changed(self:Skyfire Player Pawn Component Object Reference=@n75, Delegate=Create Event()) {is=title K2Node_RemoveDelegate}
   > Unbind Event from On Teleport to Loot Room Time Changed(self:Skyfire Player Pawn Component Object Reference=@n75, Delegate=Create Event()) {is=title K2Node_RemoveDelegate}
   > Unbind Event from On Player Entered Loot Room(self:Skyfire Player Pawn Component Object Reference=@n75, Delegate=Create Event()) {is=title K2Node_RemoveDelegate}
   > Unbind Event from On Exit Initial Prison Time Changed(self:Skyfire Player Pawn Component Object Reference=@n75, Delegate=Create Event()) {is=title K2Node_RemoveDelegate}
   > Unbind Event from On in Mothership Changed(self:Skyfire Player Pawn Component Object Reference=@n75, Delegate=Create Event()) {is=title K2Node_RemoveDelegate}
   n76 = Get!(execute=@n74.then_1) -> (then, else, CachedSkyfireMutator:Fort Athena Mutator Skyfire Object Reference)
   > Unbind Event from On Skyfire Phase Changed(self=@n76, Delegate=Create Event())
   Clear and Invalidate Timer by Handle(execute=@n74.then_2, Handle=Get TimerHandleTimeOrbsSpawnDelay())
   > Clear and Invalidate Timer by Handle(Handle=Get TimerHandleVaultOrbsSpawnDelay())
   > Clear and Invalidate Timer by Handle(Handle=Get TimerHandleMessageDisplayTime())
   > Clear and Invalidate Timer by Handle(Handle=Get TimerHandleFinalVaultCountdown() -> (TimerHandleFinalVaultCountdown:Timer Handle Structure))
   n77 = Cast To FortPlayerController(execute=@n74.then_3, Object=GetOwningPlayer())
   > Unbind Event from On Fort Pawn Changed Event(self:Fort Player Controller Object Reference=@n77, Delegate=Create Event()) {is=title K2Node_RemoveDelegate}
   > CleanupMessaging(self:Skyfire Alien Message Widget Object Reference=[Get SecondaryMessageWidget(), Get PrimaryMessageWidget()])
   n78 = Set bHasActiveContent(execute=@n74.then_4, bHasActiveContent=false)
   > SetHasActiveContent(bNewActive=@n78)
   > Set bIsFinalMessage(bIsFinalMessage=false)
   > Set bStartedFlavorTextDisplay(bStartedFlavorTextDisplay=false)
   > UpdateGlyphMaterial(self=Get PrimaryMessageWidget(), NewMaterial=&M_UI_Skyfire_Glyphs_Inst)
   > SetVisibility(InVisibility=Collapsed)
   Event MutatorReady() {is=title K2Node_Event}
   > n79 = Get()
   > Bind Event to On Skyfire Phase Changed(self:Fort Athena Mutator Skyfire Object Reference=@n79, Delegate=Create Event()) {is=title K2Node_AddDelegate}
   > BindToPawnComponent() {is=title K2Node_CallFunction}
   > Set bShowPerfFlavorText(bShowPerfFlavorText:bool=IsPerformanceFlavorTextEnabled!(self:Fort Athena Mutator Skyfire Object Reference=@n79) -> (ReturnValue:bool)) -> (Output_Get:bool)
   > Set bAllowedToShowFlavorText(bAllowedToShowFlavorText:bool=Get bShowPerfFlavorText()) -> (Output_Get:bool)
   > Set FlavorTextDisplayThreshold(FlavorTextDisplayThreshold:int=Truncate(A=GetFlavorTextCountdownThreshold!(self:Fort Athena Mutator Skyfire Object Reference=@n79) -> (ReturnValue:float))) -> (Output_Get:int=0)
   > n80 = Cast To FortPlayerController(Object=GetOwningPlayer())
   n81 = Create Event()
   > Unbind Event from On Fort Pawn Changed Event(self=@n80, Delegate=@n81)
   > Bind Event to On Fort Pawn Changed Event(self:Fort Player Controller Object Reference=@n80, Delegate=@n81) {is=title K2Node_AddDelegate}
   Event Destruct()
   > Parent: Destruct()
   > EndSkyfireMesssaging() {is=title K2Node_CallFunction}
   SelectColor(A=@n52.Color1, B=Get Color_TierReachedFill() -> (Color_TierReachedFill:Linear Color Structure), bPickA=@n51)
  ]}
 GetFlavorText {nodes [
   GetFlavorText!() -> (then, inrarity:EFortRarity Enum=Common) {is=title K2Node_FunctionEntry}
   > Clear(TargetArray:Array of Texts=Get PerfMessageArray() -> (PerfMessageArray:Array of Texts))
   > n1 = GetDataTableRowNames(Table=Get FinalMessageDataTable() -> (FinalMessageDataTable:Data Table Object Reference))
   > n2 = For Each Loop(Array:Array of Names=@n1) -> (Array Element:Name)
   > n3 = Get Data Table Row!(execute:Exec, DataTable:Data Table Object Reference=Get FinalMessageDataTable(), RowName:Name=@n2) -> (then:Exec, RowNotFound:Exec, ReturnValue:Skyfire Final Message Struct Structure)
   n4 = Break Skyfire Final Message Struct!(Skyfire_FinalMessageStruct:"Skyfire Final Message Struct Structure (by ref)"=@n3) -> (LootLevel_4_#12:EFortRarity Enum, PerformanceMessage_7_#13:Text) {is=title K2Node_BreakStruct}
   > Branch(Condition="Equal (Enum)"(A:EFortRarity Enum=Get inrarity() -> (inrarity:EFortRarity Enum=Common), B:EFortRarity Enum=@n4))
   > Branch(Condition=NOT Boolean(A=TextIsEmpty(InText=@n4.PerformanceMessage_7_#13)))
   > Add(TargetArray:Array of Texts=Get PerfMessageArray(), NewItem:"Text (by ref)"=@n4.PerformanceMessage_7_#13) {is=title K2Node_CallArrayFunction}
   n5 = Length(TargetArray:Array of Texts=Get PerfMessageArray())
   n6 = int - int(A=@n5, B=1) {is=title K2Node_CallFunction}
   Return Node!(
    execute=@n2.Completed
    PerformanceMessage:Text="Get (a copy)"(Array:Array of Texts=Get PerfMessageArray(), Dimension 1=RandomIntegerInRange!(Min:int=0, Max:int=@n6) -> (ReturnValue:int=0)) -> (Output:Text)
    Success:bool="integer > integer"(A=@n5, B=0)
   )
  ]}
 HandleSkyfireGameplayMessage {nodes [
   n1 = Inputs!() -> (ExecIn:Exec, AlienWidget:Skyfire Alien Message Widget Object Reference, MessageText:Text)
   > n2 = Branch(Condition=TextIsEmpty(InText=@n1.MessageText))
   > n3 = UpdateMessageDisplay!(execute, self:Skyfire Alien Message Widget Object Reference=@n1, MessageText:Text) -> (then)
   Outputs!(ExecOut:Exec=[@n3.then, @n4.then])
   n4 = UpdateMessageDisplay(execute=@n2.else, self=@n1, MessageText=@n1.MessageText)
  ]}
 K2_HasActiveContent {nodes [
   Has Active Content() {is=title K2Node_FunctionEntry}
   > Return Node!(execute, ReturnValue:bool=Get bHasActiveContent() -> (bHasActiveContent:bool))
  ]}
 OverrideMessageData {nodes [
   n1 = OverrideMessageData!() -> (then, Current Message:SkyfireGameplayMessageTypeEnum Enum=NewEnumerator0, TierRarity:EFortRarity Enum=Common) {is=title K2Node_FunctionEntry}
   n2 = Find(
    TargetMap:Map of SkyfireGameplayMessageTypeEnum Enums to Skyfire Gameplay Message Data Structures=Get SkyfireGameplayMessages() -> (SkyfireGameplayMessages:Map of SkyfireGameplayMessageTypeEnum Enums to Skyfire Gameplay Message Data Structures=NewEnumerator0)
    Key:"SkyfireGameplayMessageTypeEnum Enum (by ref)"=@n1
   ) -> (Value:Skyfire Gameplay Message Data Structure)
   > n3 = Branch(Condition=@n2.ReturnValue)
   > n4 = GetRarityData(Rarity=@n1.TierRarity)
   n5 = Break Skyfire Gameplay Message Data(SkyfireGameplayMessageData=@n2) -> (
    IncludeStarsTierIndicator_27_#7=~
    IconAddShadow_13_#9=~
    MessageDisplayTime_24_#11=~
   )
   > n6 = OverridePrimaryMessageData!(
    execute
    self
    CurrentMessage:SkyfireGameplayMessageTypeEnum Enum=@n1
    CurrentRarityData:Fort Rarity Item Data Structure=@n4
    PrimaryMessageData:Text=@n5
   ) -> (then, OverwrittenPrimaryMessage:Text)
   > n7 = OverrideSecondaryMessage!(
    execute
    self
    CurrentMessage:SkyfireGameplayMessageTypeEnum Enum=@n1
    SecondaryMessageText:Text=@n5.SecondaryMessageText_4_#6
    CurrentRarity:EFortRarity Enum=@n1.TierRarity
   ) -> (then, OverwrittenSecondaryMessage:Text)
   > Return Node!(
    execute
    OverrideMessageData:Skyfire Gameplay Message Data Structure=Make Skyfire Gameplay Message Data!(
     PrimaryMessageText_2_#5:Text=@n6
     SecondaryMessageText_4_#6:Text=@n7
     IncludeStarsTierIndicator_27_#7:bool=@n5.IncludeStarsTierIndicator_27_#7
     DisplayIcon_7_#8:Texture 2D Object Reference=@n5.DisplayIcon_7_#8
     IconAddShadow_13_#9:bool=@n5.IconAddShadow_13_#9
     QueuePriority_19_#10:int=@n5.QueuePriority_19_#10
     MessageDisplayTime_24_#11:float=@n5.MessageDisplayTime_24_#11
    ) -> (SkyfireGameplayMessageData:Skyfire Gameplay Message Data Structure) {is=title K2Node_MakeStruct}
   )
   FortShippingLog!(
    execute=@n3.else
    WorldContextObject:Object Reference=Self-Reference()
    InString:String=SKYFIRE Gameplay Message data was not found in map
    bLogAsWarning:bool=true
   ) -> (then)
   > Return Node(OverrideMessageData=Get DefaultErrorMessage() -> (DefaultErrorMessage:Skyfire Gameplay Message Data Structure))
  ]}
 OverridePrimaryMessageData {nodes [
   n1 = OverridePrimaryMessageData!() -> (
    then
    CurrentMessage:SkyfireGameplayMessageTypeEnum Enum=NewEnumerator0
    CurrentRarityData:Fort Rarity Item Data Structure
    PrimaryMessageData:Text
   ) {is=title K2Node_FunctionEntry}
   > n2 = Switch on SkyfireGameplayMessageTypeEnum!(execute, Selection:SkyfireGameplayMessageTypeEnum Enum=@n1) -> (
    NewEnumerator0:Exec
    NewEnumerator1:Exec
    NewEnumerator2:Exec
    NewEnumerator3:Exec
    NewEnumerator4:Exec
    NewEnumerator5:Exec
    NewEnumerator7:Exec
    NewEnumerator8:Exec
   ) {is=title K2Node_SwitchEnum}
   n3 = Get SkyfireExperimentNumberMin() -> (SkyfireExperimentNumberMin:int=0)
   n4 = Get SkyfireExperimentNumberMax() -> (SkyfireExperimentNumberMax:int=0)
   n5 = Set ExpNumber_Text(
    execute=@n2.NewEnumerator1
    ExpNumber_Text:Text="To Text (Integer)"(Value=RandomIntegerInRange(Min=@n3, Max=@n4) -> (ReturnValue=0), bUseGrouping=false)
   ) -> (Output_Get:Text)
   > Return Node!(
    execute
    OverwrittenPrimaryMessage:Text=Format Text!(Format="LOCTEXT(\"<GameplayMessagePrimary>Begin Experiment ##{RandomNumber}</>\")", RandomNumber:Text=@n5) -> (Result)
   )
   Return Node(execute=[@n2.NewEnumerator0, @n2.NewEnumerator2, @n2.NewEnumerator3, @n2.NewEnumerator4], OverwrittenPrimaryMessage=@n1.PrimaryMessageData)
   Return Node(
    execute=@n2.NewEnumerator5
    OverwrittenPrimaryMessage=Format Text!(
     Format="LOCTEXT(\"<GameplayMessagePrimary>{RarityName} Vault Tier Access Granted</>\")"
     RarityName:Text=Break Fort Rarity Item Data(FortRarityItemData=@n1.CurrentRarityData) -> (Radius=~, Falloff=~, Brightness=~, Roughness=~, Glow=~)
    ) -> (Result)
   )
   Return Node(
    execute=@n2.NewEnumerator7
    OverwrittenPrimaryMessage=Format Text!(Format="LOCTEXT(\"<GameplayMessageSpecial>{Message}</>\")", Message:Text=Get PerfText() -> (PerfText:Text)) -> (Result)
   )
  ]}
 OverrideSecondaryMessage {nodes [
   n1 = OverrideSecondaryMessage!() -> (then, CurrentMessage:SkyfireGameplayMessageTypeEnum Enum=NewEnumerator0, SecondaryMessageText:Text, CurrentRarity:EFortRarity Enum=Common) {is=title K2Node_FunctionEntry}
   > n2 = Switch on SkyfireGameplayMessageTypeEnum(Selection=@n1)
   Return Node!(
    execute=[@n2.NewEnumerator0, @n2.NewEnumerator1, @n2.NewEnumerator2, @n2.NewEnumerator3, @n2.NewEnumerator4, @n2.NewEnumerator8]
    OverwrittenSecondaryMessage:Text=@n1.SecondaryMessageText
   )
   n3 = Find(
    TargetMap:Map of EFortRarity Enums to Integers=Get RarityToOrbsCollectedMap() -> (RarityToOrbsCollectedMap=Common)
    Key:"EFortRarity Enum (by ref)"=@n1.CurrentRarity
   ) -> (Value:int)
   n4 = "To Text (Integer)"(Value=@n3)
   n5 = Get NumVaultOrbsCollectedFallbackText() -> (NumVaultOrbsCollectedFallbackText:Text)
   Return Node(
    execute=@n2.NewEnumerator5
    OverwrittenSecondaryMessage=Format Text!(
     Format="LOCTEXT(\"<GameplayMessageSecondary>{NumVaultOrbsCollected} {NumVaultOrbsCollected}|plural(one=Vault Orb,other=Vault Orbs) Collected</>\")"
     NumVaultOrbsCollected:Text=Select(Option 0:Text=@n5, Option 1:Text=@n4, Index=@n3.ReturnValue) -> (ReturnValue:Text)
    ) -> (Result)
   )
   Return Node(
    execute=@n2.NewEnumerator7
    OverwrittenSecondaryMessage=Format Text!(
     Format="LOCTEXT(\"<GameplayMessageSecondary>Experiment ##{n} Complete. Subject Performance:</>\")"
     n:Text=Get ExpNumber_Text() -> (ExpNumber_Text:Text)
    ) -> (Result)
   )
  ]}
 Set Cached Skyfire Player Pawn Component {nodes [
   n1 = Set Cached Skyfire Player Pawn Component!() -> (then, InSkyfirePlayerPawnComponent:Skyfire Player Pawn Component Object Reference) {is=title K2Node_FunctionEntry}
   > n2 = Sequence!(execute) -> (then_0, then_1)
   n3 = Set CachedSkyfirePlayerPawnComponent(execute=@n2.then_1, CachedSkyfirePlayerPawnComponent:Skyfire Player Pawn Component Object Reference=@n1) -> (Output_Get:Skyfire Player Pawn Component Object Reference)
   > Is Valid(InputObject=@n3)
   n4 = Create Event()
   > Bind Event to On Exit Initial Prison Time Changed(self:Skyfire Player Pawn Component Object Reference=@n3, Delegate=@n4) {is=title K2Node_AddDelegate}
   n5 = Create Event()
   > Bind Event to On Earned Loot Rarity Changed(self:Skyfire Player Pawn Component Object Reference=@n3, Delegate=@n5) {is=title K2Node_AddDelegate}
   n6 = Create Event()
   > Bind Event to On Exit Loot Room Time Changed(self:Skyfire Player Pawn Component Object Reference=@n3, Delegate=@n6) {is=title K2Node_AddDelegate}
   n7 = Create Event()
   > Bind Event to On Player Entered Loot Room(self:Skyfire Player Pawn Component Object Reference=@n3, Delegate=@n7) {is=title K2Node_AddDelegate}
   n8 = Create Event()
   > Bind Event to On in Mothership Changed(self:Skyfire Player Pawn Component Object Reference=@n3, Delegate=@n8) {is=title K2Node_AddDelegate}
   > EnableSkyfireMessages(InSkyfirePlayerComponent=@n3)
   n9 = Get CachedSkyfirePlayerPawnComponent()
   Is Valid(exec=@n2.then_0, InputObject=@n9)
   > Unbind Event from On Exit Initial Prison Time Changed(self=@n9, Delegate=@n4)
   > Unbind Event from On Earned Loot Rarity Changed(self=@n9, Delegate=@n5)
   > Unbind Event from On Exit Loot Room Time Changed(self=@n9, Delegate=@n6)
   > Unbind Event from On Player Entered Loot Room(self=@n9, Delegate=@n7)
   > Unbind Event from On in Mothership Changed(self=@n9, Delegate=@n8)
  ]}
 TimerToLootRoom {nodes [
   TimerToLootRoom() {is=title K2Node_FunctionEntry}
   > Branch(Condition=Get bTimerToLootRoomSet() -> (bTimerToLootRoomSet:bool))
   > n1 = Get!(execute) -> (then, else, CachedSkyfirePlayerPawnComponent:Skyfire Player Pawn Component Object Reference)
   > Branch(Condition="float < float"!(A:double=GetTimeUntilTeleportToLootRoom!(self:Skyfire Player Pawn Component Object Reference=@n1) -> (ReturnValue:float), B:double=1) -> (ReturnValue:bool) {is=title K2Node_CallFunction})
   > Clear and Invalidate Timer by Handle(Handle=Get TimeUntilLootRoom_Timer() -> (TimeUntilLootRoom_Timer:Timer Handle Structure))
   > Branch(Condition="Not Equal (Enum)"!(A:EFortRarity Enum=Get CachedCurrentLootRarity() -> (CachedCurrentLootRarity=Common), B:EFortRarity Enum=Legendary) -> (ReturnValue:bool))
   > TriggerFlavorText() {is=title K2Node_CallFunction}
  ]}
 UnbindPlayerPawn {nodes [
   UnbindPlayerPawn() {is=title K2Node_FunctionEntry}
   n1 = GetOwningPlayerPawn()
   > Is Valid(InputObject=@n1)
   > n2 = Cast To FortPlayerPawn(Object=@n1)
   > Unbind Event from On Pawn Component Attached Event(self=@n2, Delegate=Create Event())
  ]}
]
variables [
 bAllowedToShowFlavorText {type=boolean, default=true}
 bAreGameplayMessagesComplete {type=boolean}
 bHasActiveContent {type=boolean}
 bIsFinalMessage {type=boolean}
 bShowPerfFlavorText {type=boolean, default=true}
 bStartedFlavorTextDisplay {type=boolean}
 bTimerToLootRoomSet {type=boolean}
 CachedCurrentLootRarity {type=EFortRarity, default=Common}
 CachedSkyfirePlayerPawnComponent {type=SkyfirePlayerPawnComponent, default=null}
 Color_TierReachedFill {type=LinearColor, default=<rgba 1 1 1 1>}
 Color_TierUnobtainedFill {type=LinearColor, default=<rgba 0 0 0 0>}
 ColorVaultTimerDefault {type=SlateColor, default {colorUseRule=UseColor_Specified, specifiedColor=<rgba 1 1 1 1>}}
 ColorVaultTimerWarningThreshold {type=SlateColor, default {colorUseRule=UseColor_Specified, specifiedColor=<rgba 1 0 0.05 1>}}
 DefaultErrorMessage {
  type=SkyfireGameplayMessageData
  default {
   displayIcon=/Game/FrontEnd/Store/Cards/Assets/T_IconScreen_Question
   iconAddShadow=true
   includeStarsTierIndicator=false
   messageDisplayTime=6
   primaryMessageText="<GameplayMessagePrimary>Unknown Transmission</>"
   queuePriority=0
   secondaryMessageText="<GameplayMessageSecondary>Unable to Translate</>"
  }
 }
 DelaySkyfireEnterLootVaultMessageTime {type=number, default=2}
 DelaySkyfireStartMessageTime {type=number, default=4}
 ExpNumber_Text {type=string, default=""}
 FeelText {type=string, default=""}
 FinalVaultTimerUpdateRate {type=number, default=0.2}
 FlavorTextDisplayThreshold {type=integer, default=8}
 GameplayMessageQueue {type="array<SkyfireGameplayMessageData>", default []}
 IsFinalVaultTimerStarted {type=boolean}
 LargeGlyphSize {type=number, default=44}
 NumVaultOrbsCollectedFallbackText {type=string, default=""}
 PerfText {type=string, default=""}
 QueueHighestPriorityIndex {type=integer}
 RarityToOrbsCollectedMap {default {common=0, epic=3, legendary=4, rare=2, uncommon=1}}
 ReadyForNewMessage {type=boolean, default=true}
 RefTimeOrbIntialSpawnDelay {
  type=ScalableFloat
  default {registryType=None, value=1, curve {curveTable=/Skyfire/DataTables/OverrideGameData_Skyfire, rowName=Skyfire.TimeOrb.InitialSpawnDelay}}
 }
 RefTimeToLoadFromPrisonToGamespace {
  type=ScalableFloat
  default {registryType=None, value=1, curve {curveTable=&OverrideGameData_Skyfire, rowName=Skyfire.TimeToLoadFromPrisonToGamespace}}
 }
 RefVaultOrbIntialSpawnDelay {
  type=ScalableFloat
  default {registryType=None, value=1, curve {curveTable=&OverrideGameData_Skyfire, rowName=Skyfire.LootOrb.InitialSpawnDelay}}
 }
 SkyfireExperimentNumberMax {type=integer, default=9998}
 SkyfireExperimentNumberMin {type=integer, default=1000}
 SkyfireGameplayMessages {default {
   EnterLootVault {
    displayIcon=null
    iconAddShadow=true
    includeStarsTierIndicator=false
    messageDisplayTime=4
    primaryMessageText="[PH] Override Needed"
    queuePriority=1
    secondaryMessageText="[PH] Override Needed"
   }
   FlavorText1 {
    displayIcon=null
    iconAddShadow=false
    includeStarsTierIndicator=false
    messageDisplayTime=7
    primaryMessageText=""
    queuePriority=1
    secondaryMessageText="<GameplayMessageSecondary></>"
   }
   NONE {
    displayIcon=null
    iconAddShadow=true
    includeStarsTierIndicator=false
    messageDisplayTime=6
    primaryMessageText=""
    queuePriority=0
    secondaryMessageText=""
   }
   SkyfireStart {
    displayIcon=null
    iconAddShadow=false
    includeStarsTierIndicator=false
    messageDisplayTime=6
    primaryMessageText="[PH] Override Needed"
    queuePriority=2
    secondaryMessageText=""
   }
   TimeOrbsActivated {
    displayIcon=/Skyfire/UI/Art/T-Timer-Skyfire-128
    iconAddShadow=true
    includeStarsTierIndicator=false
    messageDisplayTime=6
    primaryMessageText="<GameplayMessagePrimary>Time Orbs Added</>"
    queuePriority=4
    secondaryMessageText=""
   }
   VaultOrbCollected {
    displayIcon=null
    iconAddShadow=false
    includeStarsTierIndicator=true
    messageDisplayTime=6
    primaryMessageText="<GameplayMessagePrimary>Vault Tier Upgraded</>"
    queuePriority=3
    secondaryMessageText=""
   }
   VaultOrbsActivated {
    displayIcon=/Game/UI/Foundation/Textures/Icons/Stars/Stars-Full-128
    iconAddShadow=true
    includeStarsTierIndicator=false
    messageDisplayTime=6
    primaryMessageText="<GameplayMessagePrimary>Vault Orbs Dropping</>"
    queuePriority=5
    secondaryMessageText=""
   }
  }}
 SmallGlyphSize {type=number, default=28}
 TierIndicatorWidgets {type="array<Skyfire_HUDElement_TierRating_C>", default []}
 TimerHandleFinalVaultCountdown {type=TimerHandle}
 TimerHandleMessageDisplayTime {type=TimerHandle}
 TimerHandleTimeOrbsSpawnDelay {type=TimerHandle}
 TimerHandleVaultOrbsSpawnDelay {type=TimerHandle}
 TimeUntilLootRoom_Timer {type=TimerHandle}
 TimeUntilVaultWarningColorMaxThreshold {type=integer, default=10}
]
widgets {
 info {inheritedWidgetCount=1, parentClass=/Script/SkyfireUI.SkyfireWidgetBase, rootWidgetClass=&Overlay, widgetCount=22}
 widgets [
  Overlay_Messaging {widgetClassPath=&Overlay, widget=Overlay_Messaging}
  VerticalBox_20 {
   widgetClassPath=&VerticalBox
   parent=Overlay_Messaging
   widget=VerticalBox_20
   slot {class=OverlaySlot, properties {horizontalAlignment=HAlign_Center}}
  }
  CommonText_Clock_VaultTimeRemaining {
   widgetClassPath=/Script/CommonUI.CommonNumericTextBlock
   bIsVariable=true
   parent=VerticalBox_20
   widget=CommonText_Clock_VaultTimeRemaining
   properties {
    currentNumericValue=60
    numericType=Seconds
    text=0
    visibility=Collapsed
    colorAndOpacity {colorUseRule=UseColor_Specified, specifiedColor=<rgba 1 0 0.0497 1>}
    font {
     bForceMonospaced=false
     bMaterialIsStencil=false
     fontMaterial=null
     fontObject=&Burbank
     letterSpacing=0
     monospacedWidth=1
     size=56
     skewAmount=0
     typefaceFontName=BurbankRegular-Black
     outlineSettings {
      bApplyOutlineToDropShadows=false
      bMiteredCorners=false
      bSeparateFillAlpha=false
      outlineMaterial=null
      outlineSize=0
      outlineColor=<rgba 0 0 0 1>
     }
    }
    formattingSpecification {
     alwaysSign=false
     maximumFractionalDigits=0
     maximumIntegralDigits=324
     minimumFractionalDigits=0
     minimumIntegralDigits=3
     roundingMode=FromZero
     useGrouping=false
    }
    renderTransform {angle=0, scale=<1 1>, shear=<-8 0>, translation=<0 0>}
    shadowColorAndOpacity=<rgba 0 0 0 0.7>
    shadowOffset=<3 3>
   }
   slot {class=VerticalBoxSlot, properties {horizontalAlignment=HAlign_Center, verticalAlignment=VAlign_Top}}
  }
  VerticalBox_GameplayMessagingLayout {
   widgetClassPath=&VerticalBox
   bIsVariable=true
   parent=VerticalBox_20
   widget=VerticalBox_GameplayMessagingLayout
   properties {visibility=Collapsed}
   slot {class=VerticalBoxSlot, properties {verticalAlignment=VAlign_Top}}
  }
  SizeBox_164 {
   widgetClassPath=&SizeBox
   parent=VerticalBox_GameplayMessagingLayout
   widget=SizeBox_164
   slot {class=VerticalBoxSlot, properties {verticalAlignment=VAlign_Top}}
  }
  Overlay_Icons {
   widgetClassPath=&Overlay
   parent=SizeBox_164
   widget=Overlay_Icons
   slot {class=SizeBoxSlot, properties {horizontalAlignment=HAlign_Center, verticalAlignment=VAlign_Center}}
  }
  IconGlow_HBox {
   widgetClassPath=&HorizontalBox
   bIsVariable=true
   parent=Overlay_Icons
   widget=IconGlow_HBox
   properties {visibility=Collapsed, renderTransform {angle=0, scale=<1.5 1.5>, shear=<0 0>, translation=<0 0>}}
   slot {class=OverlaySlot, properties {horizontalAlignment=HAlign_Fill, verticalAlignment=VAlign_Fill}}
  }
  IconGlow_Left {
   widgetClassPath=&Image
   bIsVariable=true
   parent=IconGlow_HBox
   widget=IconGlow_Left
   properties {
    visibility=SelfHitTestInvisible
    brush {
     bIsDynamicallyLoaded=false
     drawAs=Image
     imageType=NoImage
     mirroring=NoMirror
     resourceName=None
     resourceObject=/Skyfire/UI/Art/M_UI_Skyfire_EnergyBacking_Text
     tiling=NoTile
     imageSize=<32 32>
     margin=<ltrb 0 0 0 0>
     outlineSettings {
      bUseBrushTransparency=false
      roundingType=HalfHeightRadius
      width=0
      color {colorUseRule=UseColor_Specified, specifiedColor=<rgba 0 0 0 0>}
      cornerRadii=<0 0 0 0>
     }
     tintColor {colorUseRule=UseColor_Specified, specifiedColor=<rgba 0.5972 0 1 1>}
     uVRegion {bIsValid=false, max=<0 0>, min=<0 0>}
    }
    colorAndOpacity=<rgba 1 1 1 0.7>
    renderTransform {angle=0, scale=<-1 1>, shear=<0 0>, translation=<0 0>}
   }
   slot {class=HorizontalBoxSlot}
  }
  IconGlow_Right {
   widgetClassPath=&Image
   bIsVariable=true
   parent=IconGlow_HBox
   widget=IconGlow_Right
   properties {
    visibility=SelfHitTestInvisible
    brush {
     bIsDynamicallyLoaded=false
     drawAs=Image
     imageType=NoImage
     mirroring=NoMirror
     resourceName=None
     resourceObject=&M_UI_Skyfire_EnergyBacking_Text
     tiling=NoTile
     imageSize=<32 32>
     margin=<ltrb 0 0 0 0>
     outlineSettings {
      bUseBrushTransparency=false
      roundingType=HalfHeightRadius
      width=0
      color {colorUseRule=UseColor_Specified, specifiedColor=<rgba 0 0 0 0>}
      cornerRadii=<0 0 0 0>
     }
     tintColor {colorUseRule=UseColor_Specified, specifiedColor=<rgba 0.5972 0 1 1>}
     uVRegion {bIsValid=false, max=<0 0>, min=<0 0>}
    }
    colorAndOpacity=<rgba 1 1 1 0.7>
   }
   slot {class=HorizontalBoxSlot}
  }
  Image_Icon_shadow {
   widgetClassPath=&Image
   bIsVariable=true
   parent=Overlay_Icons
   widget=Image_Icon_shadow
   properties {
    visibility=SelfHitTestInvisible
    brush {
     bIsDynamicallyLoaded=false
     drawAs=Image
     imageType=NoImage
     mirroring=NoMirror
     resourceName=None
     resourceObject=null
     tiling=NoTile
     imageSize=<64 64>
     margin=<ltrb 0 0 0 0>
     outlineSettings {
      bUseBrushTransparency=false
      roundingType=HalfHeightRadius
      width=0
      color {colorUseRule=UseColor_Specified, specifiedColor=<rgba 0 0 0 0>}
      cornerRadii=<0 0 0 0>
     }
     tintColor {colorUseRule=UseColor_Specified, specifiedColor=<rgba 1 1 1 1>}
     uVRegion {bIsValid=false, max=<0 0>, min=<0 0>}
    }
    colorAndOpacity=<rgba 0.5972 0 1 1>
   }
   slot {class=OverlaySlot, properties {horizontalAlignment=HAlign_Center, verticalAlignment=VAlign_Center, padding=<ltrb 2 3 0 0>}}
  }
  Image_Icon {
   widgetClassPath=&Image
   bIsVariable=true
   parent=Overlay_Icons
   widget=Image_Icon
   properties {
    visibility=SelfHitTestInvisible
    brush {
     bIsDynamicallyLoaded=false
     drawAs=Image
     imageType=NoImage
     mirroring=NoMirror
     resourceName=None
     resourceObject=null
     tiling=NoTile
     imageSize=<64 64>
     margin=<ltrb 0 0 0 0>
     outlineSettings {
      bUseBrushTransparency=false
      roundingType=HalfHeightRadius
      width=0
      color {colorUseRule=UseColor_Specified, specifiedColor=<rgba 0 0 0 0>}
      cornerRadii=<0 0 0 0>
     }
     tintColor {colorUseRule=UseColor_Specified, specifiedColor=<rgba 1 1 1 1>}
     uVRegion {bIsValid=false, max=<0 0>, min=<0 0>}
    }
    colorAndOpacity=<rgba 0.9647 0.5906 1 1>
   }
   slot {class=OverlaySlot, properties {horizontalAlignment=HAlign_Center, verticalAlignment=VAlign_Center}}
  }
  SizeBox_TierIndicator {
   widgetClassPath=&SizeBox
   bIsVariable=true
   parent=VerticalBox_GameplayMessagingLayout
   widget=SizeBox_TierIndicator
   properties {bOverride_HeightOverride=true, heightOverride=50, visibility=Collapsed}
   slot {class=VerticalBoxSlot, properties {horizontalAlignment=HAlign_Center, verticalAlignment=VAlign_Center}}
  }
  ScaleBox_97 {
   widgetClassPath=&ScaleBox
   parent=SizeBox_TierIndicator
   widget=ScaleBox_97
   slot {class=SizeBoxSlot, properties {horizontalAlignment=HAlign_Center}}
  }
  HorizontalBox-TierIndicator {
   widgetClassPath=&HorizontalBox
   parent=ScaleBox_97
   widget=HorizontalBox-TierIndicator
   slot {class=ScaleBoxSlot, properties {verticalAlignment=VAlign_Top}}
  }
  Skyfire_HUDElement_TierRating_1 {
   widgetClassPath=&/Skyfire_HUDElement_TierRating
   bIsVariable=true
   parent=HorizontalBox-TierIndicator
   widget=Skyfire_HUDElement_TierRating_1
   slot {class=HorizontalBoxSlot, properties {size {sizeRule=Fill, value=0.2}}}
  }
  Skyfire_HUDElement_TierRating_2 {
   widgetClassPath=&Skyfire_HUDElement_TierRating
   bIsVariable=true
   parent=HorizontalBox-TierIndicator
   widget=Skyfire_HUDElement_TierRating_2
   slot {class=HorizontalBoxSlot, properties {size {sizeRule=Fill, value=0.2}}}
  }
  Skyfire_HUDElement_TierRating_3 {
   widgetClassPath=&Skyfire_HUDElement_TierRating
   bIsVariable=true
   parent=HorizontalBox-TierIndicator
   widget=Skyfire_HUDElement_TierRating_3
   slot {class=HorizontalBoxSlot, properties {size {sizeRule=Fill, value=0.2}}}
  }
  Skyfire_HUDElement_TierRating_4 {
   widgetClassPath=&Skyfire_HUDElement_TierRating
   bIsVariable=true
   parent=HorizontalBox-TierIndicator
   widget=Skyfire_HUDElement_TierRating_4
   slot {class=HorizontalBoxSlot, properties {size {sizeRule=Fill, value=0.2}}}
  }
  Skyfire_HUDElement_TierRating_5 {
   widgetClassPath=&Skyfire_HUDElement_TierRating
   bIsVariable=true
   parent=HorizontalBox-TierIndicator
   widget=Skyfire_HUDElement_TierRating_5
   slot {class=HorizontalBoxSlot, properties {size {sizeRule=Fill, value=0.2}}}
  }
  SecondaryMessageWidget {
   widgetClassPath=&/Skyfire_AlienMessageWidget
   bIsVariable=true
   parent=VerticalBox_GameplayMessagingLayout
   widget=SecondaryMessageWidget
   properties {glyphSize=24, previewText="<GameplayMessageSecondary>[PH] SECONDARY TEXT</>"}
   slot {class=VerticalBoxSlot, properties {horizontalAlignment=HAlign_Center, verticalAlignment=VAlign_Center}}
  }
  PrimaryMessageWidget {
   widgetClassPath=&Skyfire_AlienMessageWidget
   bIsVariable=true
   parent=VerticalBox_GameplayMessagingLayout
   widget=PrimaryMessageWidget
   properties {previewText="<GameplayMessagePrimary>[PH] PRIMARY TEXT</>"}
   slot {class=VerticalBoxSlot, properties {horizontalAlignment=HAlign_Center, verticalAlignment=VAlign_Center, size {sizeRule=Automatic, value=0.5}}}
  }
  TutorialHighlight {widgetClassPath=&TutorialHighlight}
 ]
}
