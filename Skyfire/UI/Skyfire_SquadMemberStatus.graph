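# Squad member status widget for Skyfire that shows a mothership icon when that squadmate is aboard the mothership.
# The icon sits in a scale box inside a size box capped at a small square, and is hidden on Construct.
# The parent SkyfireSquadmateStatusWidget's mothership status event toggles it between SelfHitTestInvisible and Hidden.
# A debug print of the chosen visibility is left unconnected in the graph.
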WidgetBlueprint /Skyfire/UI/Skyfire_SquadMemberStatus
graphs [EventGraph {nodes [
   n1 = Event OnSquadMemberMothershipStatusChanged!() -> (OutputDelegate, then, bIsOnMothership:bool) {is=title K2Node_Event}
   n2 = Select(Option 0=Hidden, Option 1=SelfHitTestInvisible, Index=@n1.bIsOnMothership)
   > SetVisibility(self=Get Image_MothershipIcon() -> (Image_MothershipIcon:Image Object Reference), InVisibility=@n2)
   Event Construct()
   > SetVisibility(self=Get Image_MothershipIcon(), InVisibility=Hidden)
   PrintString!(
    execute
    InString=Append(A="LogFortSkyfire - Updating MothershipIcon visibility to ", B=Enum to String!(Enumerator:ESlateVisibility Enum=@n2) -> (ReturnValue:String))
    bPrintToScreen
    bPrintToLog
    TextColor
    Duration=20
    Key
   ) -> (then)
  ]}]
variables []
widgets {
 info {inheritedWidgetCount=1, parentClass=/Script/SkyfireUI.SkyfireSquadmateStatusWidget, rootWidgetClass=&SizeBox, widgetCount=4}
 widgets [
  SizeBox_5 {
   widgetClassPath=&SizeBox
   widget=SizeBox_5
   properties {bOverride_MaxDesiredHeight=true, bOverride_MaxDesiredWidth=true, maxDesiredHeight=48, maxDesiredWidth=48}
  }
  ScaleBox_24 {widgetClassPath=&ScaleBox, parent=SizeBox_5, widget=ScaleBox_24, slot {class=SizeBoxSlot}}
  Image_MothershipIcon {
   widgetClassPath=&Image
   bIsVariable=true
   parent=ScaleBox_24
   widget=Image_MothershipIcon
   properties {brush {
     bIsDynamicallyLoaded=false
     drawAs=Image
     imageType=NoImage
     mirroring=NoMirror
     resourceName=None
     resourceObject=&/T-Icon-Skyfire-mothership-128
     tiling=NoTile
     imageSize=<128 128>
     margin=<ltrb 0 0 0 0>
     outlineSettings {
      bUseBrushTransparency=false
      roundingType=HalfHeightRadius
      width=0
      color {colorUseRule=UseColor_Specified, specifiedColor=<rgba 0 0 0 0>}
      cornerRadii=<0 0 0 0>
     }
     tintColor {colorUseRule=UseColor_Specified, specifiedColor=<rgba 1 1 1 1>}
     uVRegion {bIsValid=false, max=<0 0>, min=<0 0>}
    }}
   slot {class=ScaleBoxSlot}
  }
  TutorialHighlight {widgetClassPath=&TutorialHighlight}
 ]
}
